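# Translucent UI material for the Autoclaim icon, with an on/off Current Status scalar parameter.
# It samples the icon texture through MF_UI_UVScaling's Scaled to Fit UVs and fades the blue channel to zero by a red-channel mask.
# The ceiled status drives that mask's lerp through a named reroute, but both inputs are red, so the toggle changes nothing.
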Material /GiftBox/UI/Art/Materials/M_UI_Autoclaim
graph {
 properties {
  bAutomaticallySetUsageInEditor=false
  bUsedWithStaticMesh=false
  blendMode=BLEND_Translucent
  materialDomain=MD_UI
  usedShadingModels=Default Lit
 }
 nodes [
  Constant3Vector() {constant=<rgba 1 1 1 0>}
  n1 = TextureSample(UVs=MaterialFunctionCall() {materialFunction=&MF_UI_UVScaling}.Scaled to Fit) {samplerSource=SSM_Clamp_WorldGroupSettings, texture=/GiftBox/UI/Art/Textures/T_UI_Icon_Autoclaim}
  LinearInterpolate(A=@n1.B, Alpha=LinearInterpolate(A=@n1.R, B=@n1.R, Alpha=NamedRerouteUsage())) {constB=0}
  NamedRerouteDeclaration(None=Ceil(None=ScalarParameter() {defaultValue=1, desc="1 = ON\r\n0 = OFF", parameterName=Current Status})) {name=Current Status, nodeColor=<rgba 0 0 0 1>}
 ]
}
