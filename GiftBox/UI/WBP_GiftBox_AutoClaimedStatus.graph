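# GiftBox UI widget that shows a static "AUTO-CLAIMED" status badge.
# An Overlay stacks a post-buffer capture widget and a translucent dark UIKit backplate behind a padded row.
# The row holds a green-tinted auto-claim material icon beside a CommonTextBlock label.
# The event graph has only unconnected PreConstruct, Construct and Tick events, so the widget carries no logic.
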WidgetBlueprint /GiftBox/UI/WBP_GiftBox_AutoClaimedStatus
graphs [EventGraph {nodes [
   Event PreConstruct()
   Event Construct()
   Event Tick()
  ]}]
variables []
widgets {
 info {inheritedWidgetCount=0, parentClass=&UserWidget, rootWidgetClass=&Overlay, widgetCount=6}
 widgets [
  Root {widgetClassPath=&Overlay, widget=Root}
  WBP_CaptureForPostBufferUpdate {
   widgetClassPath=&WBP_CaptureForPostBufferUpdate
   bIsVariable=true
   parent=Root
   slot {class=OverlaySlot, properties {horizontalAlignment=HAlign_Fill, verticalAlignment=VAlign_Fill}}
  }
  BG_Backplate {
   widgetClassPath=&WBP_UIKit_Backplate
   bIsVariable=true
   parent=Root
   widget=BG_Backplate
   properties {brightness=Dark, cornerRadius=Button, colorAndOpacity=<rgba 0.0513 0.0513 0.0513 0.466>}
   slot {class=OverlaySlot, properties {horizontalAlignment=HAlign_Fill, verticalAlignment=VAlign_Fill}}
  }
  HorizontalBox_43 {
   widgetClassPath=&HorizontalBox
   parent=Root
   widget=HorizontalBox_43
   slot {class=OverlaySlot, properties {padding=<ltrb 8 8 8 8>}}
  }
  Image_AutoClaimed {
   widgetClassPath=&Image
   bIsVariable=true
   parent=HorizontalBox_43
   widget=Image_AutoClaimed
   properties {brush {
     bIsDynamicallyLoaded=false
     drawAs=Image
     imageType=FullColor
     mirroring=NoMirror
     resourceName=None
     resourceObject=/GiftBox/UI/Art/Materials/M_UI_Autoclaim
     tiling=NoTile
     imageSize=<28 24>
     margin=<ltrb 0 0 0 0>
     outlineSettings {
      bUseBrushTransparency=false
      roundingType=HalfHeightRadius
      width=0
      color {colorUseRule=UseColor_Specified, specifiedColor=<rgba 0 0 0 0>}
      cornerRadii=<0 0 0 0>
     }
     tintColor {colorUseRule=UseColor_Specified, specifiedColor=<rgba 0.0395 1 0.2307 1>}
     uVRegion {bIsValid=false, max=<0 0>, min=<0 0>}
    }}
   slot {class=HorizontalBoxSlot, properties {horizontalAlignment=HAlign_Left, verticalAlignment=VAlign_Center}}
  }
  CommonTextBlock_90 {
   widgetClassPath=&CommonTextBlock
   parent=HorizontalBox_43
   widget=CommonTextBlock_90
   properties {
    style=&TextStyle-Base_C
    text=AUTO-CLAIMED
    font {
     bForceMonospaced=false
     bMaterialIsStencil=false
     fontMaterial=null
     fontObject=&NotoSans
     letterSpacing=0
     monospacedWidth=1
     size=24
     skewAmount=0
     typefaceFontName=NotoSans-Regular
     outlineSettings {
      bApplyOutlineToDropShadows=false
      bMiteredCorners=false
      bSeparateFillAlpha=false
      outlineMaterial=null
      outlineSize=0
      outlineColor=<rgba 0 0 0 1>
     }
    }
    shadowOffset=<0 0>
   }
   slot {class=HorizontalBoxSlot, properties {verticalAlignment=VAlign_Center, padding=<ltrb 8 0 0 0>}}
  }
 ]
}
