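# GiftBoxScreenHeader is a CommonUI widget for the gift box screen header, showing a reward image, header, subheader, body and auto-claim status.
# Helpers prefer gift box text and image over item fallbacks; setters collapse empty subheader, body or image.
# HandleGiftBoxSet drives auto-claim visibility from the item's autoSpend parameter.
# InitialState snaps main elements hidden; IntroAnim and FadeAnim stagger slide-and-fade transitions, with only the intro animating auto-claim status.
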WidgetBlueprint /GiftBox/UI/GiftBoxScreenHeader
animations [
 IntroAnim {
  binding_count=5
  length_seconds=0.5
  object=&self:IntroAnim
  sequence {bindings [
    AutoClaimStatus {
     id=#1
     possessed_class=WBP_GiftBox_AutoClaimedStatus_C
     tracks [
      {
       type=MovieSceneFloatTrack
       display_name=Render Opacity
       sections [{type=MovieSceneFloatSection, end_frame=10, start_frame=7, channels [None {type=MovieSceneScriptingFloatChannel, keys [[7,0],[10,1]]}]}]
      }
      {
       type=MovieScene2DTransformTrack
       sections [{
         type=MovieScene2DTransformSection
         end_frame=10
         start_frame=7
         channels [
          Translation.X {type=MovieSceneScriptingFloatChannel, const [0,1]}
          Translation.Y {type=MovieSceneScriptingFloatChannel, keys [[7,-16],[10,0]]}
          Angle {type=MovieSceneScriptingFloatChannel, const [0,1]}
          Scale.X {type=MovieSceneScriptingFloatChannel, const [1,1]}
          Scale.Y {type=MovieSceneScriptingFloatChannel, const [1,1]}
          Shear.X {type=MovieSceneScriptingFloatChannel, const [0,1]}
          Shear.Y {type=MovieSceneScriptingFloatChannel, const [0,1]}
         ]
        }]
      }
     ]
    }
    Text_Body {
     id=#2
     possessed_class=CommonTextBlock
     tracks [
      {
       type=MovieScene2DTransformTrack
       sections [{
         type=MovieScene2DTransformSection
         channels [
          Translation.X {type=MovieSceneScriptingFloatChannel, const [0,2]}
          Translation.Y {type=MovieSceneScriptingFloatChannel, keys [[7,-16],[10,0]]}
          Angle {type=MovieSceneScriptingFloatChannel}
          Scale.X {type=MovieSceneScriptingFloatChannel}
          Scale.Y {type=MovieSceneScriptingFloatChannel}
          Shear.X {type=MovieSceneScriptingFloatChannel}
          Shear.Y {type=MovieSceneScriptingFloatChannel}
         ]
        }]
      }
      {
       type=MovieSceneFloatTrack
       display_name=Render Opacity
       sections [{type=MovieSceneFloatSection, channels [None {type=MovieSceneScriptingFloatChannel, keys [[7,0],[10,1]]}]}]
      }
     ]
    }
    Text_SubHeader {
     id=#3
     possessed_class=CommonTextBlock
     tracks [
      {
       type=MovieScene2DTransformTrack
       sections [{
         type=MovieScene2DTransformSection
         channels [
          Translation.X {type=MovieSceneScriptingFloatChannel, const [0,2]}
          Translation.Y {type=MovieSceneScriptingFloatChannel, keys [[6,-32],[9,0]]}
          Angle {type=MovieSceneScriptingFloatChannel}
          Scale.X {type=MovieSceneScriptingFloatChannel}
          Scale.Y {type=MovieSceneScriptingFloatChannel}
          Shear.X {type=MovieSceneScriptingFloatChannel}
          Shear.Y {type=MovieSceneScriptingFloatChannel}
         ]
        }]
      }
      {
       type=MovieSceneFloatTrack
       display_name=Render Opacity
       sections [{type=MovieSceneFloatSection, channels [None {type=MovieSceneScriptingFloatChannel, keys [[6,0],[9,1]]}]}]
      }
     ]
    }
    Text_Header {
     id=#4
     possessed_class=CommonTextBlock
     tracks [
      {
       type=MovieSceneFloatTrack
       display_name=Render Opacity
       sections [{type=MovieSceneFloatSection, channels [None {type=MovieSceneScriptingFloatChannel, keys [[3,0],[6,1]]}]}]
      }
      {
       type=MovieScene2DTransformTrack
       sections [{
         type=MovieScene2DTransformSection
         channels [
          Translation.X {type=MovieSceneScriptingFloatChannel, keys [[3,128],[6,0]]}
          Translation.Y {type=MovieSceneScriptingFloatChannel, const [0,2]}
          Angle {type=MovieSceneScriptingFloatChannel}
          Scale.X {type=MovieSceneScriptingFloatChannel}
          Scale.Y {type=MovieSceneScriptingFloatChannel}
          Shear.X {type=MovieSceneScriptingFloatChannel}
          Shear.Y {type=MovieSceneScriptingFloatChannel}
         ]
        }]
      }
     ]
    }
    Image_Reward {
     id=#5
     possessed_class=CommonLazyImage
     tracks [
      {
       type=MovieSceneFloatTrack
       display_name=Render Opacity
       sections [{type=MovieSceneFloatSection, channels [None {type=MovieSceneScriptingFloatChannel, keys [[0,0],[3,1]]}]}]
      }
      {
       type=MovieScene2DTransformTrack
       sections [{
         type=MovieScene2DTransformSection
         channels [
          Translation.X {type=MovieSceneScriptingFloatChannel, keys [[0,-32],[3,0]]}
          Translation.Y {type=MovieSceneScriptingFloatChannel, const [0,2]}
          Angle {type=MovieSceneScriptingFloatChannel}
          Scale.X {type=MovieSceneScriptingFloatChannel}
          Scale.Y {type=MovieSceneScriptingFloatChannel}
          Shear.X {type=MovieSceneScriptingFloatChannel}
          Shear.Y {type=MovieSceneScriptingFloatChannel}
         ]
        }]
      }
     ]
    }
   ]}
 }
 InitialState {
  binding_count=4
  length_seconds=0.05
  object=&self:InitialState
  sequence {bindings [
    Text_Body {
     id=#2
     possessed_class=CommonTextBlock
     tracks [
      {
       type=MovieScene2DTransformTrack
       sections [{
         type=MovieScene2DTransformSection
         channels [
          Translation.X {type=MovieSceneScriptingFloatChannel, const [0,1]}
          Translation.Y {type=MovieSceneScriptingFloatChannel, const [-32,1]}
          Angle {type=MovieSceneScriptingFloatChannel}
          Scale.X {type=MovieSceneScriptingFloatChannel}
          Scale.Y {type=MovieSceneScriptingFloatChannel}
          Shear.X {type=MovieSceneScriptingFloatChannel}
          Shear.Y {type=MovieSceneScriptingFloatChannel}
         ]
        }]
      }
      {
       type=MovieSceneFloatTrack
       display_name=Render Opacity
       sections [{type=MovieSceneFloatSection, channels [None {type=MovieSceneScriptingFloatChannel, const [0,1]}]}]
      }
     ]
    }
    Text_SubHeader {
     id=#3
     possessed_class=CommonTextBlock
     tracks [
      {
       type=MovieScene2DTransformTrack
       sections [{
         type=MovieScene2DTransformSection
         channels [
          Translation.X {type=MovieSceneScriptingFloatChannel, const [0,1]}
          Translation.Y {type=MovieSceneScriptingFloatChannel, const [-32,1]}
          Angle {type=MovieSceneScriptingFloatChannel}
          Scale.X {type=MovieSceneScriptingFloatChannel}
          Scale.Y {type=MovieSceneScriptingFloatChannel}
          Shear.X {type=MovieSceneScriptingFloatChannel}
          Shear.Y {type=MovieSceneScriptingFloatChannel}
         ]
        }]
      }
      {
       type=MovieSceneFloatTrack
       display_name=Render Opacity
       sections [{type=MovieSceneFloatSection, channels [None {type=MovieSceneScriptingFloatChannel, const [0,1]}]}]
      }
     ]
    }
    Text_Header {
     id=#6
     possessed_class=CommonTextBlock
     tracks [
      {
       type=MovieSceneFloatTrack
       display_name=Render Opacity
       sections [{type=MovieSceneFloatSection, channels [None {type=MovieSceneScriptingFloatChannel, const [0,1]}]}]
      }
      {
       type=MovieScene2DTransformTrack
       sections [{
         type=MovieScene2DTransformSection
         channels [
          Translation.X {type=MovieSceneScriptingFloatChannel, const [128,1]}
          Translation.Y {type=MovieSceneScriptingFloatChannel, const [0,1]}
          Angle {type=MovieSceneScriptingFloatChannel}
          Scale.X {type=MovieSceneScriptingFloatChannel}
          Scale.Y {type=MovieSceneScriptingFloatChannel}
          Shear.X {type=MovieSceneScriptingFloatChannel}
          Shear.Y {type=MovieSceneScriptingFloatChannel}
         ]
        }]
      }
     ]
    }
    Image_Reward {
     id=#7
     possessed_class=CommonLazyImage
     tracks [
      {
       type=MovieSceneFloatTrack
       display_name=Render Opacity
       sections [{type=MovieSceneFloatSection, channels [None {type=MovieSceneScriptingFloatChannel, const [0,1]}]}]
      }
      {
       type=MovieScene2DTransformTrack
       sections [{
         type=MovieScene2DTransformSection
         channels [
          Translation.X {type=MovieSceneScriptingFloatChannel, const [-32,1]}
          Translation.Y {type=MovieSceneScriptingFloatChannel, const [0,1]}
          Angle {type=MovieSceneScriptingFloatChannel}
          Scale.X {type=MovieSceneScriptingFloatChannel}
          Scale.Y {type=MovieSceneScriptingFloatChannel}
          Shear.X {type=MovieSceneScriptingFloatChannel}
          Shear.Y {type=MovieSceneScriptingFloatChannel}
         ]
        }]
      }
     ]
    }
   ]}
 }
 FadeAnim {
  binding_count=4
  length_seconds=0.5
  object=&self:FadeAnim
  sequence {bindings [
    Image_Reward {
     id=#8
     possessed_class=CommonLazyImage
     tracks [
      {
       type=MovieSceneFloatTrack
       display_name=Render Opacity
       sections [{type=MovieSceneFloatSection, channels [None {type=MovieSceneScriptingFloatChannel, keys [[0,1],[3,0]]}]}]
      }
      {
       type=MovieScene2DTransformTrack
       sections [{
         type=MovieScene2DTransformSection
         channels [
          Translation.X {type=MovieSceneScriptingFloatChannel, keys [[0,0],[3,-64]]}
          Translation.Y {type=MovieSceneScriptingFloatChannel, const [0,2]}
          Angle {type=MovieSceneScriptingFloatChannel}
          Scale.X {type=MovieSceneScriptingFloatChannel}
          Scale.Y {type=MovieSceneScriptingFloatChannel}
          Shear.X {type=MovieSceneScriptingFloatChannel}
          Shear.Y {type=MovieSceneScriptingFloatChannel}
         ]
        }]
      }
     ]
    }
    Text_Body {
     id=#2
     possessed_class=CommonTextBlock
     tracks [
      {
       type=MovieScene2DTransformTrack
       sections [{
         type=MovieScene2DTransformSection
         channels [
          Translation.X {type=MovieSceneScriptingFloatChannel, keys [[7,0],[10,-128]]}
          Translation.Y {type=MovieSceneScriptingFloatChannel, const [0,2]}
          Angle {type=MovieSceneScriptingFloatChannel}
          Scale.X {type=MovieSceneScriptingFloatChannel}
          Scale.Y {type=MovieSceneScriptingFloatChannel}
          Shear.X {type=MovieSceneScriptingFloatChannel}
          Shear.Y {type=MovieSceneScriptingFloatChannel}
         ]
        }]
      }
      {
       type=MovieSceneFloatTrack
       display_name=Render Opacity
       sections [{type=MovieSceneFloatSection, channels [None {type=MovieSceneScriptingFloatChannel, keys [[7,1],[10,0]]}]}]
      }
     ]
    }
    Text_SubHeader {
     id=#3
     possessed_class=CommonTextBlock
     tracks [
      {
       type=MovieScene2DTransformTrack
       sections [{
         type=MovieScene2DTransformSection
         channels [
          Translation.X {type=MovieSceneScriptingFloatChannel, keys [[6,0],[9,-128]]}
          Translation.Y {type=MovieSceneScriptingFloatChannel, const [0,2]}
          Angle {type=MovieSceneScriptingFloatChannel}
          Scale.X {type=MovieSceneScriptingFloatChannel}
          Scale.Y {type=MovieSceneScriptingFloatChannel}
          Shear.X {type=MovieSceneScriptingFloatChannel}
          Shear.Y {type=MovieSceneScriptingFloatChannel}
         ]
        }]
      }
      {
       type=MovieSceneFloatTrack
       display_name=Render Opacity
       sections [{type=MovieSceneFloatSection, channels [None {type=MovieSceneScriptingFloatChannel, keys [[6,1],[9,0]]}]}]
      }
     ]
    }
    Text_Header {
     id=#9
     possessed_class=CommonTextBlock
     tracks [
      {
       type=MovieSceneFloatTrack
       display_name=Render Opacity
       sections [{type=MovieSceneFloatSection, channels [None {type=MovieSceneScriptingFloatChannel, keys [[3,1],[6,0]]}]}]
      }
      {
       type=MovieScene2DTransformTrack
       sections [{
         type=MovieScene2DTransformSection
         channels [
          Translation.X {type=MovieSceneScriptingFloatChannel, keys [[3,0],[6,-128]]}
          Translation.Y {type=MovieSceneScriptingFloatChannel, const [0,2]}
          Angle {type=MovieSceneScriptingFloatChannel}
          Scale.X {type=MovieSceneScriptingFloatChannel}
          Scale.Y {type=MovieSceneScriptingFloatChannel}
          Shear.X {type=MovieSceneScriptingFloatChannel}
          Shear.Y {type=MovieSceneScriptingFloatChannel}
         ]
        }]
      }
     ]
    }
   ]}
 }
]
graphs [
 EventGraph {nodes [
   Event PreConstruct()
   > Set Header Color(Header Color:Slate Color Structure=Get ColorAndOpacity!(self:Text Object Reference=Get Text_Header() -> (Text_Header:Common Text Object Reference)) -> (ColorAndOpacity:Slate Color Structure)) -> (Output_Get:Slate Color Structure)
   > Set SubHeader Color(SubHeader Color:Slate Color Structure=Get ColorAndOpacity(self=Get Text_SubHeader() -> (Text_SubHeader:Common Text Object Reference))) -> (Output_Get:Slate Color Structure)
   > Set Body Color(Body Color:Slate Color Structure=Get ColorAndOpacity(self=Get Text_Body() -> (Text_Body:Common Text Object Reference))) -> (Output_Get:Slate Color Structure)
  ]}
 Get Color Override {nodes [
   n1 = Get Color Override!() -> (then, Color:Slate Color Structure, Color Override:Slate Color Structure, Use Color Override:bool=~) {is=title K2Node_FunctionEntry}
   > Return Node!(
    execute
    NewParam1:Slate Color Structure=Select(Option 0:Slate Color Structure=@n1, Option 1:Slate Color Structure=@n1.Color Override, Index=@n1.Use Color Override) -> (ReturnValue:Slate Color Structure)
   )
  ]}
 Get Header Text {nodes [
   n1 = Get Header Text!() -> (then, Gift Box Text:Text, Item Text:Text) {is=title K2Node_FunctionEntry}
   n2 = "To String (Text)"(InText=SelectText(A=@n1.Item Text, B=@n1, bPickA=TextIsEmpty(InText=@n1)))
   n3 = "To String (Text)"(InText=MakeLiteralText(Value="LOCTEXT(\"\\r\\n\")"))
   > Return Node!(execute, Result:Text="To Text (String)"(InString=Replace(SourceString=@n2, From="\\n", To=@n3, SearchCase=CaseSensitive)))
  ]}
 Get Reward Image {nodes [
   n1 = Get Reward Image!() -> (then, Gift Box Image:Soft Object Reference, Reward Image:Soft Object Reference) {is=title K2Node_FunctionEntry}
   > Return Node(Result:Soft Object Reference=Select(
     Option 0:Soft Object Reference=@n1.Reward Image
     Option 1:Soft Object Reference=@n1
     Index=IsValidSoftObjectReference(SoftObjectReference=@n1)
    ) -> (ReturnValue:Soft Object Reference))
  ]}
 HandleGiftBoxSet {nodes [
   n1 = HandleGiftBoxSet!() -> (then, GiftBox:Gift Box VM Object Reference) {is=title K2Node_FunctionEntry}
   > n2 = Get!(execute:Exec, self:Gift Box VM Object Reference=@n1) -> (then:Exec, else:Exec, GiftBoxItem:Fort Gift Box Item Object Reference)
   n3 = TryGetParamBool!(self:Fort Gift Box Item Object Reference=@n2, ParamName:String=autoSpend) -> (OutValue:bool, ReturnValue:bool)
   > n4 = Branch(Condition=@n3)
   > SetVisibility(
    self=Get AutoClaimStatus() -> (AutoClaimStatus:WBP Gift Box Auto Claimed Status Object Reference)
    InVisibility=Select(Option 0:ESlateVisibility Enum=Collapsed, Option 1:ESlateVisibility Enum=HitTestInvisible, Index=@n3.ReturnValue) -> (ReturnValue:ESlateVisibility Enum)
   )
   SetVisibility(execute=[@n4.else, @n2.else], self=Get AutoClaimStatus(), InVisibility=Collapsed)
  ]}
 HandleInitialState {nodes [
   HandleInitialState() {is=title K2Node_FunctionEntry}
   > PlayAnimationTimeRange!(
    execute
    self:User Widget Object Reference
    InAnimation:Widget Animation Object Reference=Get InitialState() -> (InitialState:Widget Animation Object Reference)
    StartAtTime:float=0
    EndAtTime:float=0
    NumLoopsToPlay:int=1
    PlayMode:EUMGSequencePlayMode Enum
    PlaybackSpeed:float=1
    bRestoreState:bool=false
   ) -> (then, ReturnValue:Widget Animation Handle Structure)
  ]}
 HandleIntro {nodes [
   HandleIntro() {is=title K2Node_FunctionEntry}
   > StopAnimation(InAnimation=Get FadeAnim() -> (FadeAnim:Widget Animation Object Reference))
   > PlayAnimationForward(InAnimation=Get IntroAnim() -> (IntroAnim:Widget Animation Object Reference), PlaybackSpeed=1.5)
  ]}
 HandleOutro {nodes [
   HandleOutro() {is=title K2Node_FunctionEntry}
   > PlayAnimationForward(InAnimation=Get FadeAnim(), PlaybackSpeed=1.5)
  ]}
 Set Body Text {nodes [
   n1 = Set Body Text!() -> (then, Text:Text) {is=title K2Node_FunctionEntry}
   n2 = Get Text_Body()
   > "SetText (Text)"(self=@n2, InText=@n1)
   > SetVisibility(self=@n2, InVisibility=Select(Option 0=SelfHitTestInvisible, Option 1=Collapsed, Index=TextIsEmpty(InText=@n1)))
  ]}
 Set Reward Image {nodes [
   n1 = Set Reward Image!() -> (then, Image:Soft Object Reference) {is=title K2Node_FunctionEntry}
   n2 = Get Image_Reward() -> (Image_Reward:Common Lazy Image Object Reference)
   > SetBrushFromLazyDisplayAsset!(execute, self:Common Lazy Image Object Reference=@n2, LazyObject:"Soft Object Reference (by ref)"=@n1, bMatchTextureSize:bool=false) -> (then)
   > SetVisibility(self=@n2, InVisibility=Select(Option 0=Collapsed, Option 1=SelfHitTestInvisible, Index=IsValidSoftObjectReference(SoftObjectReference=@n1)))
  ]}
 Set Sub Header Text {nodes [
   n1 = Set Sub Header Text!() -> (then, Text:Text) {is=title K2Node_FunctionEntry}
   n2 = Get Text_SubHeader()
   > "SetText (Text)"(self=@n2, InText=@n1)
   > SetVisibility(self=@n2, InVisibility=Select(Option 0=SelfHitTestInvisible, Option 1=Collapsed, Index=TextIsEmpty(InText=@n1)))
  ]}
]
variables [
 Badge Color {type=SlateColor, default {colorUseRule=UseColor_Specified, specifiedColor=<rgba 1 0 1 1>}}
 Body Color {type=SlateColor, default {colorUseRule=UseColor_Specified, specifiedColor=<rgba 1 0 1 1>}}
 Header Color {type=SlateColor, default {colorUseRule=UseColor_Specified, specifiedColor=<rgba 1 0 1 1>}}
 SubHeader Color {type=SlateColor, default {colorUseRule=UseColor_Specified, specifiedColor=<rgba 1 0 1 1>}}
]
widgets {
 info {inheritedWidgetCount=0, parentClass=&CommonUserWidget, rootWidgetClass=&HorizontalBox, widgetCount=7}
 widgets [
  HB_Content {widgetClassPath=&HorizontalBox, widget=HB_Content}
  Image_Reward {
   widgetClassPath=/Script/CommonUI.CommonLazyImage
   bIsVariable=true
   parent=HB_Content
   widget=Image_Reward
   properties {
    visibility=HitTestInvisible
    brush {
     bIsDynamicallyLoaded=false
     drawAs=Image
     imageType=NoImage
     mirroring=NoMirror
     resourceName=None
     resourceObject=/Game/UI/Foundation/Textures/Icons/Items/T-FNBR-BattlePass
     tiling=NoTile
     imageSize=<128 128>
     margin=<ltrb 0 0 0 0>
     outlineSettings {
      bUseBrushTransparency=false
      roundingType=HalfHeightRadius
      width=0
      color {colorUseRule=UseColor_Specified, specifiedColor=<rgba 0 0 0 0>}
      cornerRadii=<0 0 0 0>
     }
     tintColor {colorUseRule=UseColor_Specified, specifiedColor=<rgba 1 1 1 1>}
     uVRegion {bIsValid=false, max=<0 0>, min=<0 0>}
    }
   }
   slot {class=HorizontalBoxSlot, properties {horizontalAlignment=HAlign_Left, verticalAlignment=VAlign_Top, padding=<ltrb 0 0 16 0>}}
  }
  VerticalBox_1 {
   widgetClassPath=&VerticalBox
   parent=HB_Content
   widget=VerticalBox_1
   slot {class=HorizontalBoxSlot, properties {verticalAlignment=VAlign_Center, size {sizeRule=Fill, value=1}}}
  }
  Text_Header {
   widgetClassPath=&CommonTextBlock
   bIsVariable=true
   parent=VerticalBox_1
   widget=Text_Header
   properties {
    autoWrapText=true
    lineHeightPercentage=0.85
    text="{HEADER STRING}"
    textTransformPolicy=ToUpper
    font {
     bForceMonospaced=false
     bMaterialIsStencil=false
     fontMaterial=null
     fontObject=&HeadingNow
     letterSpacing=0
     monospacedWidth=1
     size=32.25
     skewAmount=0
     typefaceFontName=76 Bold Italic
     outlineSettings {
      bApplyOutlineToDropShadows=false
      bMiteredCorners=false
      bSeparateFillAlpha=false
      outlineMaterial=null
      outlineSize=0
      outlineColor=<rgba 0 0 0 1>
     }
    }
   }
   slot {class=VerticalBoxSlot, properties {verticalAlignment=VAlign_Top}}
  }
  Text_SubHeader {
   widgetClassPath=&CommonTextBlock
   bIsVariable=true
   parent=VerticalBox_1
   widget=Text_SubHeader
   properties {
    autoWrapText=true
    text="[ph]SUBTITLE LOREM IPSUM DOLOR"
    visibility=Collapsed
    font {
     bForceMonospaced=false
     bMaterialIsStencil=false
     fontMaterial=null
     fontObject=null
     letterSpacing=0
     monospacedWidth=1
     size=24
     skewAmount=0
     typefaceFontName=None
     outlineSettings {
      bApplyOutlineToDropShadows=false
      bMiteredCorners=false
      bSeparateFillAlpha=false
      outlineMaterial=null
      outlineSize=0
      outlineColor=<rgba 0 0 0 1>
     }
    }
   }
   slot {class=VerticalBoxSlot, properties {verticalAlignment=VAlign_Top, padding=<ltrb 0 8 0 8>}}
  }
  Text_Body {
   widgetClassPath=&CommonTextBlock
   bIsVariable=true
   parent=VerticalBox_1
   widget=Text_Body
   properties {
    autoWrapText=true
    flowDirectionPreference=LeftToRight
    lineHeightPercentage=1.29
    text="{AssetBodyString} Lorem IpsumDolor Amit Lorem  Ipsum Dolor Amit Lorem Ipsum Dolor Amit"
    colorAndOpacity {colorUseRule=UseColor_Specified, specifiedColor=<rgba 0.5776 0.5776 0.5776 1>}
    font {
     bForceMonospaced=false
     bMaterialIsStencil=false
     fontMaterial=null
     fontObject=&HeadingNow
     letterSpacing=0
     monospacedWidth=1
     size=18
     skewAmount=0
     typefaceFontName=74 Regular
     outlineSettings {
      bApplyOutlineToDropShadows=false
      bMiteredCorners=false
      bSeparateFillAlpha=false
      outlineMaterial=null
      outlineSize=0
      outlineColor=<rgba 0 0 0 1>
     }
    }
    shapedTextOptions {bOverride_TextFlowDirection=true, bOverride_TextShapingMethod=false, textFlowDirection=LeftToRight, textShapingMethod=Auto}
   }
   slot {class=VerticalBoxSlot, properties {verticalAlignment=VAlign_Top, padding=<ltrb 0 8 0 0>}}
  }
  AutoClaimStatus {
   widgetClassPath=&/WBP_GiftBox_AutoClaimedStatus
   bIsVariable=true
   parent=VerticalBox_1
   widget=AutoClaimStatus
   slot {class=VerticalBoxSlot, properties {horizontalAlignment=HAlign_Left, padding=<ltrb 0 16 0 0>}}
  }
 ]
}
