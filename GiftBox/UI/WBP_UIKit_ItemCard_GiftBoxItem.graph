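# Gift box item card widget, extending the UIKit item card base with variant ownership, currency quantity visibility and per-type animation scale.
# SetIsOwned reads variant counts from FortItemVM, showing an owned/total highlight for partial ownership; the owned check indicator resolves to Collapsed either way.
# ToggleDivider creates or removes a divider in the custom named slot, adjusting grid column, padding and card size.
# Unhover reaches the parent only without touch or selection.
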WidgetBlueprint /GiftBox/UI/WBP_UIKit_ItemCard_GiftBoxItem
graphs [
 DetermineQuantityVisibility {nodes [
   n1 = DetermineQuantityVisibility!() -> (then, Quantity:int, ItemType:EFortItemType Enum=WorldItem) {is=title K2Node_FunctionEntry}
   n2 = "Equal (Enum)"(A:EFortItemType Enum=@n1.ItemType, B:EFortItemType Enum=Currency)
   n3 = "integer > integer"(A=@n1, B=1)
   > Return Node!(execute, Visibility:ESlateVisibility Enum=Select(Option 0=Collapsed, Option 1=HitTestInvisible, Index=AND Boolean(A=@n3, B=@n2)))
  ]}
 EventGraph {nodes [
   n1 = Event PreConstruct()
   > Parent: PreConstruct(IsDesignTime=@n1.IsDesignTime)
   Event On Unhovered()
   > n2 = Branch(Condition=IsUsingTouch(PlayerController=GetOwningPlayer()))
   n3 = Branch(execute=@n2.else, Condition=GetSelected!(self:Common Button Base Object Reference) -> (ReturnValue:bool))
   Parent: On Unhovered(execute=@n3.else) {is=title K2Node_CallParentFunction}
   ToggleDivider!(execute, self, ShouldShowDivider:bool=true) -> (then)
  ]}
 OnItemTypeSet {nodes [
   n1 = OnItemTypeSet!() -> (then, ItemType:"EFortItemType Enum (by ref)"=WorldItem) {is=title K2Node_FunctionEntry}
   n2 = Find(
    TargetMap:"Map of EFortItemType Enums to Float (double-precision)s"=Get AnimationScaleOverridesPerItemType() -> (AnimationScaleOverridesPerItemType:"Map of EFortItemType Enums to Float (double-precision)s"=WorldItem)
    Key:"EFortItemType Enum (by ref)"=@n1
   ) -> (Value:double)
   > Branch(Condition=@n2.ReturnValue)
   > SetItemAnimationScale!(execute, self:WBP UIKit Item Card Base Object Reference, Item Scale:"Float (double-precision) (by ref)"=@n2) -> (then)
  ]}
 ResetVariantsInfo {nodes [
   ResetVariantsInfo() {is=title K2Node_FunctionEntry}
   > Set Visibility(Visibility=Collapsed, self=Get O_OwnedAmount() -> (O_OwnedAmount:Overlay Object Reference))
   > Set Visibility(Visibility=Collapsed, self=Get StatusIndicator_Owned() -> (StatusIndicator_Owned:WBP UIKit Status Indicator Object Reference))
  ]}
 Set Variants Info {nodes [
   n1 = Set Variants Info!() -> (then, Owned:int, Total:int) {is=title K2Node_FunctionEntry}
   > Set OwnedVariants(OwnedVariants:int=@n1) -> (Output_Get:int=0)
   > Set TotalVariants(TotalVariants:int=@n1.Total) -> (Output_Get:int=0)
   > ResetVariantsInfo() {is=title K2Node_CallFunction}
   > SetVariantsCountVisibility()
   > SetOwnedVariantsText()
  ]}
 SetIsOwned {nodes [
   n1 = SetIsOwned!() -> (then, IsOwned:bool=~) {is=title K2Node_FunctionEntry}
   n2 = Get FortItemVM() -> (FortItemVM:Fort Item VM Object Reference)
   > Is Valid(InputObject=@n2)
   n3 = GetOwnedVariants!(self:Fort Item VM Object Reference=@n2) -> (TotalVariants:int=0, ReturnValue:int=0)
   > Set Variants Info!(execute, self, Owned:int=@n3.ReturnValue, Total:int=@n3) -> (then) {is=title K2Node_CallFunction}
   > n4 = Set IsOwned(IsOwned:bool=@n1) -> (Output_Get:bool)
   n5 = NOT Boolean(A=Get PartiallyOwnsVariants() -> (PartiallyOwnsVariants:bool))
   > Set Visibility(Visibility=Select(Option 0=Collapsed, Option 1=Collapsed, Index=AND Boolean(A=@n4, B=@n5)), self=Get StatusIndicator_Owned())
  ]}
 SetOwnedVariantsText {nodes [
   SetOwnedVariantsText() {is=title K2Node_FunctionEntry}
   > SetText!(
    execute
    self:WBP UIKit Highlight Base Object Reference=Get Highlight_OwnedAmount() -> (Highlight_OwnedAmount:WBP UIKit Highlight Object Reference)
    Text:Text=Format Text!(Format="LOCTEXT(\"{0}/{1}\")", 0:int=Get OwnedVariants() -> (OwnedVariants:int=0), 1:int=Get TotalVariants() -> (TotalVariants:int=0)) -> (Result)
   ) -> (then)
  ]}
 SetVariantsCountVisibility {nodes [
   SetVariantsCountVisibility() {is=title K2Node_FunctionEntry}
   > n1 = Branch(Condition="Not Equal (Integer)"(A=Get OwnedVariants() -> (OwnedVariants=0)))
   > n2 = Branch(Condition="Equal (Integer)"(A=Get OwnedVariants() -> (OwnedVariants=0), B=Get TotalVariants() -> (TotalVariants=0)))
   > Set Visibility(Visibility=Collapsed, self=Get StatusIndicator_Owned())
   > Set PartiallyOwnsVariants(PartiallyOwnsVariants:bool=false) -> (Output_Get:bool)
   Set Visibility(execute=@n2.else, Visibility=SelfHitTestInvisible, self=Get O_OwnedAmount())
   > Set PartiallyOwnsVariants(PartiallyOwnsVariants=true)
   Set PartiallyOwnsVariants(execute=@n1.else, PartiallyOwnsVariants=false)
  ]}
 ToggleDivider {nodes [
   n1 = ToggleDivider!() -> (then, ShouldShowDivider:bool) {is=title K2Node_FunctionEntry}
   > n2 = Branch(Condition=@n1)
   > n3 = Get!(execute:Exec) -> (then:Exec, else:Exec, WidgetDivider:WBP UIKit Divider Giftbox Object Reference)
   > n4 = RemoveFromParent(self=@n3)
   n5 = Construct Object from Class(execute=[@n4.then, @n3.else], Class=&/WBP_UIKit_Divider_Giftbox.WBP_UIKit_Divider_Giftbox_C) -> (ReturnValue:WBP UIKit Divider Giftbox Object Reference)
   > Set WidgetDivider(WidgetDivider:WBP UIKit Divider Giftbox Object Reference=@n5) -> (Output_Get:WBP UIKit Divider Giftbox Object Reference)
   > AddChild(
    self=Get Slot_Custom() -> (Slot_Custom:Named Slot Object Reference)
    Content=Get WidgetDivider() -> (WidgetDivider:WBP UIKit Divider Giftbox Object Reference)
   )
   n6 = SlotAsGridSlot(Widget=Get Slot_Custom())
   > Set Column!(execute, Column:int=1, self:Grid Slot Object Reference=@n6) -> (then, Output_Get:int=0)
   > Set!(execute, Padding_Left:float=16, Padding_Top:float=56, Padding_Right:float=0, Padding_Bottom:float=56, self:Grid Slot Object Reference=@n6) -> (then, Output_Get:Margin Structure)
   > SetItemCardSize!(execute, self:WBP UIKit Item Card Base Object Reference, ItemSize:Vector 2D Structure=(X=148,Y=160)) -> (then)
   n7 = Get(execute=@n2.else)
   > n8 = RemoveFromParent(self=@n7)
   Set Column(execute=[@n8.then, @n7.else], Column=0, self=@n6) -> (Output_Get=0)
   > Set(Padding_Left=0, Padding_Top=0, Padding_Right=0, Padding_Bottom=0, self=@n6)
   > SetItemCardSize(ItemSize=(X=128,Y=160))
   Get Root() -> (Root:Grid Panel Object Reference)
  ]}
]
variables [
 AnimationScaleOverridesPerItemType {default {accountResource=0.7, currency=0.7}}
 IsOwned {type=boolean}
 OwnedVariants {type=integer}
 PartiallyOwnsVariants {type=boolean}
 TotalVariants {type=integer}
 WidgetDivider {type=WBP_UIKit_Divider_Giftbox_C, default=null}
]
widgets {
 info {
  inheritedWidgetCount=0
  namedSlotCount=1
  parentClass=/FortUILibrary/Systems/Items/WBP_UIKit_ItemCard_Base.WBP_UIKit_ItemCard_Base_C
  rootWidgetClass=null
  widgetCount=8
 }
 widgets [
  GridPanel_Owned {widgetClassPath=&GridPanel, widget=GridPanel_Owned, properties {columnFill [1], rowFill [1]}}
  OverlayGradient_Currency {
   widgetClassPath=&CommonBorder
   bIsVariable=true
   parent=GridPanel_Owned
   widget=OverlayGradient_Currency
   properties {
    renderOpacity=0.85
    background {
     bIsDynamicallyLoaded=false
     drawAs=Image
     imageType=NoImage
     mirroring=NoMirror
     resourceName=None
     resourceObject=/FortUILibrary/Systems/Items/MaterialInstances/MI_UI_Preset_Overlay
     tiling=NoTile
     imageSize=<32 32>
     margin=<ltrb 0 0 0 0>
     outlineSettings {
      bUseBrushTransparency=false
      roundingType=HalfHeightRadius
      width=0
      color {colorUseRule=UseColor_Specified, specifiedColor=<rgba 0 0 0 0>}
      cornerRadii=<0 0 0 0>
     }
     tintColor {colorUseRule=UseColor_Specified, specifiedColor=<rgba 1 1 1 1>}
     uVRegion {bIsValid=false, max=<0 0>, min=<0 0>}
    }
   }
   slot {class=GridSlot}
  }
  O_OwnedAmount {
   widgetClassPath=&Overlay
   bIsVariable=true
   parent=GridPanel_Owned
   widget=O_OwnedAmount
   slot {class=GridSlot, properties {horizontalAlignment=HAlign_Right, verticalAlignment=VAlign_Top, padding=<ltrb 8 8 8 0>}}
  }
  Highlight_OwnedAmount {
   widgetClassPath=/FortUILibrary/UIKit/Highlight/WBP_UIKit_Highlight
   bIsVariable=true
   parent=O_OwnedAmount
   widget=Highlight_OwnedAmount
   properties {highlightSize=Compact, highlightType=Social}
   slot {class=OverlaySlot}
  }
  StatusIndicator_Owned {
   widgetClassPath=/FortUILibrary/UIKit/StatusIndicator/WBP_UIKit_StatusIndicator
   bIsVariable=true
   parent=GridPanel_Owned
   widget=StatusIndicator_Owned
   properties {icon=/FortUILibrary/Textures/MaterialInstances/MI_UIKit_Icon_Check, overrideStatusIndicatorVisibility=true}
   slot {class=GridSlot, properties {horizontalAlignment=HAlign_Right, verticalAlignment=VAlign_Top, padding=<ltrb 0 8 8 0>}}
  }
  HB_Currency {
   widgetClassPath=&HorizontalBox
   bIsVariable=true
   parent=GridPanel_Owned
   widget=HB_Currency
   properties {visibility=HitTestInvisible}
   slot {class=GridSlot, properties {horizontalAlignment=HAlign_Right, verticalAlignment=VAlign_Bottom, padding=<ltrb 8 0 8 8>}}
  }
  Txt_CurrencyAmount {
   widgetClassPath=&CommonTextBlock
   bIsVariable=true
   parent=HB_Currency
   widget=Txt_CurrencyAmount
   properties {
    justification=Right
    scrollStyle=&TextScrollStyle-Base_C
    text=00
    font {
     bForceMonospaced=false
     bMaterialIsStencil=false
     fontMaterial=null
     fontObject=&NotoSans
     letterSpacing=0
     monospacedWidth=1
     size=24
     skewAmount=0
     typefaceFontName=NotoSans-Regular
     outlineSettings {
      bApplyOutlineToDropShadows=false
      bMiteredCorners=false
      bSeparateFillAlpha=false
      outlineMaterial=null
      outlineSize=0
      outlineColor=<rgba 0 0 0 1>
     }
    }
    shadowOffset=<0 0>
   }
   slot {class=HorizontalBoxSlot}
  }
  Spacer_61 {
   widgetClassPath=&Spacer
   parent=GridPanel_Owned
   widget=Spacer_61
   properties {visibility=Visible}
   slot {class=GridSlot, properties {padding=<ltrb -8 0 -8 0>}}
  }
 ]
}
