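# Gift Box screen widget that presents gifted rewards in a tabbed, paginated list with item preview and claim, equip, customize and go-to-locker actions.
# Tab and item selection update the list and preview, while claim or go-to-locker plays collected and fade animations before closing or navigating.
# Custom side widgets and backgrounds load asynchronously, with a pending-path check discarding stale background loads.
# Button clicks and seen items are reported to analytics.
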WidgetBlueprint /GiftBox/UI/GiftBoxScreen
animations [
 Anim_Intro {
  binding_count=8
  length_seconds=0.7
  object=&self:Anim_Intro
  sequence {
   bindings [
    ListView_Items {
     id=#1
     possessed_class=CommonListView
     tracks [{
       type=MovieSceneFloatTrack
       display_name=Render Opacity
       sections [{type=MovieSceneFloatSection, end_frame=14, start_frame=0, channels [None {type=MovieSceneScriptingFloatChannel, keys [[0,0],[14,1]]}]}]
      }]
    }
    HB_BtnActions {
     id=#2
     possessed_class=HorizontalBox
     tracks [
      {
       type=MovieSceneByteTrack
       display_name=Visibility
       sections [{type=MovieSceneByteSection, end_frame=14, start_frame=0, channels [None {type=MovieSceneScriptingByteChannel, keys [[0,3],[14,4]]}]}]
      }
      {
       type=MovieScene2DTransformTrack
       sections [{
         type=MovieScene2DTransformSection
         end_frame=0
         start_frame=0
         channels [
          Translation.X {type=MovieSceneScriptingFloatChannel, const [0,1]}
          Translation.Y {type=MovieSceneScriptingFloatChannel, const [0,1]}
          Angle {type=MovieSceneScriptingFloatChannel}
          Scale.X {type=MovieSceneScriptingFloatChannel}
          Scale.Y {type=MovieSceneScriptingFloatChannel}
          Shear.X {type=MovieSceneScriptingFloatChannel}
          Shear.Y {type=MovieSceneScriptingFloatChannel}
         ]
        }]
      }
      {
       type=MovieSceneFloatTrack
       display_name=Render Opacity
       sections [{type=MovieSceneFloatSection, end_frame=14, start_frame=0, channels [None {type=MovieSceneScriptingFloatChannel, const [1,1]}]}]
      }
     ]
    }
    ItemHeaderWidget {
     id=#3
     possessed_class=ItemPresentationGenericRewardHeader_C
     tracks [
      {
       type=MovieScene2DTransformTrack
       sections [{
         type=MovieScene2DTransformSection
         channels [
          Translation.X {type=MovieSceneScriptingFloatChannel, keys [[2,-128],[10,0]]}
          Translation.Y {type=MovieSceneScriptingFloatChannel, const [0,2]}
          Angle {type=MovieSceneScriptingFloatChannel}
          Scale.X {type=MovieSceneScriptingFloatChannel}
          Scale.Y {type=MovieSceneScriptingFloatChannel}
          Shear.X {type=MovieSceneScriptingFloatChannel, keys [[2,12],[10,0]]}
          Shear.Y {type=MovieSceneScriptingFloatChannel, const [0,2]}
         ]
        }]
      }
      {
       type=MovieSceneFloatTrack
       display_name=Render Opacity
       sections [{type=MovieSceneFloatSection, channels [None {type=MovieSceneScriptingFloatChannel, keys [[2,0],[10,1]]}]}]
      }
     ]
    }
    ItemDescription {
     id=#4
     possessed_class=WBP_UIKit_DescriptionWithSubtitle_C
     tracks [
      {
       type=MovieSceneFloatTrack
       display_name=Render Opacity
       sections [{type=MovieSceneFloatSection, channels [None {type=MovieSceneScriptingFloatChannel, keys [[2,0],[10,1]]}]}]
      }
      {
       type=MovieScene2DTransformTrack
       sections [{
         type=MovieScene2DTransformSection
         channels [
          Translation.X {type=MovieSceneScriptingFloatChannel, keys [[2,128],[10,0]]}
          Translation.Y {type=MovieSceneScriptingFloatChannel, const [0,2]}
          Angle {type=MovieSceneScriptingFloatChannel}
          Scale.X {type=MovieSceneScriptingFloatChannel}
          Scale.Y {type=MovieSceneScriptingFloatChannel}
          Shear.X {type=MovieSceneScriptingFloatChannel, keys [[2,-12],[10,0]]}
          Shear.Y {type=MovieSceneScriptingFloatChannel, const [0,2]}
         ]
        }]
      }
     ]
    }
    O_Buttons {
     id=#5
     possessed_class=Overlay
     tracks [
      {
       type=MovieSceneFloatTrack
       display_name=Render Opacity
       sections [{type=MovieSceneFloatSection, channels [None {type=MovieSceneScriptingFloatChannel, keys [[6,0],[14,1]]}]}]
      }
      {
       type=MovieScene2DTransformTrack
       sections [{
         type=MovieScene2DTransformSection
         channels [
          Translation.X {type=MovieSceneScriptingFloatChannel, const [0,3]}
          Translation.Y {type=MovieSceneScriptingFloatChannel, keys [[0,64],[5,64],[13,0]]}
          Angle {type=MovieSceneScriptingFloatChannel}
          Scale.X {type=MovieSceneScriptingFloatChannel}
          Scale.Y {type=MovieSceneScriptingFloatChannel}
          Shear.X {type=MovieSceneScriptingFloatChannel}
          Shear.Y {type=MovieSceneScriptingFloatChannel}
         ]
        }]
      }
     ]
    }
    Text_ItemsCollected {
     id=#6
     possessed_class=CommonTextBlock
     tracks [{
       type=MovieSceneByteTrack
       display_name=Visibility
       sections [{type=MovieSceneByteSection, end_frame=0, start_frame=0, channels [None {type=MovieSceneScriptingByteChannel, const [2,1]}]}]
      }]
    }
    Button_LeftScrollArrow {
     id=#7
     possessed_class=CircleArrowButton_C
     tracks [
      {
       type=MovieScene2DTransformTrack
       sections [{
         type=MovieScene2DTransformSection
         channels [
          Translation.X {type=MovieSceneScriptingFloatChannel, keys [[7,64],[13,0]]}
          Translation.Y {type=MovieSceneScriptingFloatChannel, const [0,2]}
          Angle {type=MovieSceneScriptingFloatChannel}
          Scale.X {type=MovieSceneScriptingFloatChannel}
          Scale.Y {type=MovieSceneScriptingFloatChannel}
          Shear.X {type=MovieSceneScriptingFloatChannel}
          Shear.Y {type=MovieSceneScriptingFloatChannel}
         ]
        }]
      }
      {
       type=MovieSceneFloatTrack
       display_name=Render Opacity
       sections [{type=MovieSceneFloatSection, channels [None {type=MovieSceneScriptingFloatChannel, keys [[7,0],[13,1]]}]}]
      }
     ]
    }
    Button_RightScrollArrow {
     id=#8
     possessed_class=CircleArrowButton_C
     tracks [
      {
       type=MovieScene2DTransformTrack
       sections [{
         type=MovieScene2DTransformSection
         channels [
          Translation.X {type=MovieSceneScriptingFloatChannel, keys [[7,-64],[13,0]]}
          Translation.Y {type=MovieSceneScriptingFloatChannel, const [0,2]}
          Angle {type=MovieSceneScriptingFloatChannel}
          Scale.X {type=MovieSceneScriptingFloatChannel}
          Scale.Y {type=MovieSceneScriptingFloatChannel}
          Shear.X {type=MovieSceneScriptingFloatChannel}
          Shear.Y {type=MovieSceneScriptingFloatChannel}
         ]
        }]
      }
      {
       type=MovieSceneFloatTrack
       display_name=Render Opacity
       sections [{type=MovieSceneFloatSection, channels [None {type=MovieSceneScriptingFloatChannel, keys [[7,0],[13,1]]}]}]
      }
     ]
    }
   ]
   tracks [{
     type=MovieSceneEventTrack
     sections [{
       type=MovieSceneEventTriggerSection
       channels [None {type=MovieSceneScriptingEventChannel, const ["<Struct 'MovieSceneEvent' (0x46883850a00) {}>", 1]}]
      }]
    }]
  }
 }
 Anim_Collected {
  binding_count=2
  length_seconds=0.3
  object=&self:Anim_Collected
  sequence {bindings [
    Text_ItemsCollected {
     id=#9
     possessed_class=TextBlock
     tracks [
      {
       type=MovieSceneByteTrack
       display_name=Visibility
       sections [{type=MovieSceneByteSection, channels [None {type=MovieSceneScriptingByteChannel, const [3,1]}]}]
      }
      {
       type=MovieScene2DTransformTrack
       sections [{
         type=MovieScene2DTransformSection
         channels [
          Translation.X {type=MovieSceneScriptingFloatChannel}
          Translation.Y {type=MovieSceneScriptingFloatChannel, keys [[3,64],[6,0]]}
          Angle {type=MovieSceneScriptingFloatChannel}
          Scale.X {type=MovieSceneScriptingFloatChannel}
          Scale.Y {type=MovieSceneScriptingFloatChannel}
          Shear.X {type=MovieSceneScriptingFloatChannel}
          Shear.Y {type=MovieSceneScriptingFloatChannel}
         ]
        }]
      }
      {
       type=MovieSceneFloatTrack
       display_name=Render Opacity
       sections [{type=MovieSceneFloatSection, channels [None {type=MovieSceneScriptingFloatChannel, keys [[3,0],[6,1]]}]}]
      }
     ]
    }
    HB_BtnActions {
     id=#10
     possessed_class=HorizontalBox
     tracks [
      {
       type=MovieScene2DTransformTrack
       sections [{
         type=MovieScene2DTransformSection
         channels [
          Translation.X {type=MovieSceneScriptingFloatChannel}
          Translation.Y {type=MovieSceneScriptingFloatChannel, keys [[0,0],[3,-64]]}
          Angle {type=MovieSceneScriptingFloatChannel}
          Scale.X {type=MovieSceneScriptingFloatChannel}
          Scale.Y {type=MovieSceneScriptingFloatChannel}
          Shear.X {type=MovieSceneScriptingFloatChannel}
          Shear.Y {type=MovieSceneScriptingFloatChannel}
         ]
        }]
      }
      {
       type=MovieSceneByteTrack
       display_name=Visibility
       sections [{type=MovieSceneByteSection, channels [None {type=MovieSceneScriptingByteChannel, keys [[0,3],[4,2]]}]}]
      }
      {
       type=MovieSceneFloatTrack
       display_name=Render Opacity
       sections [{type=MovieSceneFloatSection, channels [None {type=MovieSceneScriptingFloatChannel, keys [[0,1],[3,0]]}]}]
      }
     ]
    }
   ]}
 }
 Anim_InitialStateSet {
  binding_count=7
  length_seconds=0.05
  object=&self:Anim_InitialStateSet
  sequence {bindings [
    ItemHeaderWidget {
     id=#3
     possessed_class=ItemPresentationGenericRewardHeader_C
     tracks [
      {
       type=MovieScene2DTransformTrack
       sections [{
         type=MovieScene2DTransformSection
         channels [
          Translation.X {type=MovieSceneScriptingFloatChannel, const [-128,1]}
          Translation.Y {type=MovieSceneScriptingFloatChannel, const [0,1]}
          Angle {type=MovieSceneScriptingFloatChannel}
          Scale.X {type=MovieSceneScriptingFloatChannel}
          Scale.Y {type=MovieSceneScriptingFloatChannel}
          Shear.X {type=MovieSceneScriptingFloatChannel}
          Shear.Y {type=MovieSceneScriptingFloatChannel}
         ]
        }]
      }
      {
       type=MovieSceneFloatTrack
       display_name=Render Opacity
       sections [{type=MovieSceneFloatSection, channels [None {type=MovieSceneScriptingFloatChannel, const [0,1]}]}]
      }
     ]
    }
    Button_LeftScrollArrow {
     id=#11
     possessed_class=CircleArrowButton_C
     tracks [{
       type=MovieSceneFloatTrack
       display_name=Render Opacity
       sections [{type=MovieSceneFloatSection, end_frame=0, start_frame=0, channels [None {type=MovieSceneScriptingFloatChannel, const [0,1]}]}]
      }]
    }
    ItemDescription {
     id=#12
     possessed_class=WBP_UIKit_DescriptionWithSubtitle_C
     tracks [
      {
       type=MovieSceneFloatTrack
       display_name=Render Opacity
       sections [{type=MovieSceneFloatSection, channels [None {type=MovieSceneScriptingFloatChannel, const [0,1]}]}]
      }
      {
       type=MovieScene2DTransformTrack
       sections [{
         type=MovieScene2DTransformSection
         channels [
          Translation.X {type=MovieSceneScriptingFloatChannel, const [128,1]}
          Translation.Y {type=MovieSceneScriptingFloatChannel, const [0,1]}
          Angle {type=MovieSceneScriptingFloatChannel}
          Scale.X {type=MovieSceneScriptingFloatChannel}
          Scale.Y {type=MovieSceneScriptingFloatChannel}
          Shear.X {type=MovieSceneScriptingFloatChannel}
          Shear.Y {type=MovieSceneScriptingFloatChannel}
         ]
        }]
      }
     ]
    }
    Txt_Equipped {
     id=#13
     possessed_class=CommonTextBlock
     tracks [{
       type=MovieSceneFloatTrack
       display_name=Render Opacity
       sections [{type=MovieSceneFloatSection, end_frame=0, start_frame=0, channels [None {type=MovieSceneScriptingFloatChannel, const [0,1]}]}]
      }]
    }
    Button_RightScrollArrow {
     id=#14
     possessed_class=CircleArrowButton_C
     tracks [{
       type=MovieSceneFloatTrack
       display_name=Render Opacity
       sections [{type=MovieSceneFloatSection, end_frame=0, start_frame=0, channels [None {type=MovieSceneScriptingFloatChannel, const [0,1]}]}]
      }]
    }
    O_Buttons {
     id=#5
     possessed_class=Overlay
     tracks [
      {
       type=MovieSceneFloatTrack
       display_name=Render Opacity
       sections [{type=MovieSceneFloatSection, channels [None {type=MovieSceneScriptingFloatChannel, const [0,1]}]}]
      }
      {
       type=MovieScene2DTransformTrack
       sections [{
         type=MovieScene2DTransformSection
         channels [
          Translation.X {type=MovieSceneScriptingFloatChannel, const [0,1]}
          Translation.Y {type=MovieSceneScriptingFloatChannel, const [64,1]}
          Angle {type=MovieSceneScriptingFloatChannel}
          Scale.X {type=MovieSceneScriptingFloatChannel}
          Scale.Y {type=MovieSceneScriptingFloatChannel}
          Shear.X {type=MovieSceneScriptingFloatChannel}
          Shear.Y {type=MovieSceneScriptingFloatChannel}
         ]
        }]
      }
     ]
    }
    ListView_Items {
     id=#15
     possessed_class=CommonListView
     tracks [{
       type=MovieSceneFloatTrack
       display_name=Render Opacity
       sections [{type=MovieSceneFloatSection, end_frame=0, start_frame=0, channels [None {type=MovieSceneScriptingFloatChannel, const [0,1]}]}]
      }]
    }
   ]}
 }
 Anim_Equipped {
  binding_count=1
  length_seconds=0.75
  object=&self:Anim_Equipped
  sequence {bindings [Txt_Equipped {
     id=#16
     possessed_class=CommonTextBlock
     tracks [
      {
       type=MovieSceneFloatTrack
       display_name=Render Opacity
       sections [{type=MovieSceneFloatSection, channels [None {type=MovieSceneScriptingFloatChannel, keys [[0,0],[1,1],[15,0]]}]}]
      }
      {
       type=MovieScene2DTransformTrack
       sections [{
         type=MovieScene2DTransformSection
         channels [
          Translation.X {type=MovieSceneScriptingFloatChannel, const [0,2]}
          Translation.Y {type=MovieSceneScriptingFloatChannel, keys [[0,160],[15,0]]}
          Angle {type=MovieSceneScriptingFloatChannel}
          Scale.X {type=MovieSceneScriptingFloatChannel}
          Scale.Y {type=MovieSceneScriptingFloatChannel}
          Shear.X {type=MovieSceneScriptingFloatChannel}
          Shear.Y {type=MovieSceneScriptingFloatChannel}
         ]
        }]
      }
      {
       type=MovieSceneColorTrack
       display_name=Color and Opacity
       sections [{
         type=MovieSceneColorSection
         channels [
          Color.R {type=MovieSceneScriptingFloatChannel, keys [[0,1],[15,0]]}
          Color.G {type=MovieSceneScriptingFloatChannel, keys [[0,1],[15,0.6654]]}
          Color.B {type=MovieSceneScriptingFloatChannel, const [1,2]}
          Color.A {type=MovieSceneScriptingFloatChannel, const [1,2]}
         ]
        }]
      }
      {
       type=MovieSceneByteTrack
       display_name=Visibility
       sections [{type=MovieSceneByteSection, channels [None {type=MovieSceneScriptingByteChannel, keys [[0,3],[15,2]]}]}]
      }
     ]
    }]}
 }
 Anim_Fade {
  binding_count=6
  length_seconds=0.5
  object=&self:Anim_Fade
  sequence {bindings [
    WBP_GiftBox_Unlocked {
     id=#17
     possessed_class=WBP_GiftBox_Unlocked_C
     tracks [{
       type=MovieSceneFloatTrack
       display_name=Render Opacity
       sections [{type=MovieSceneFloatSection, end_frame=10, start_frame=0, channels [None {type=MovieSceneScriptingFloatChannel, keys [[0,1],[10,0]]}]}]
      }]
    }
    ListView_Items {
     id=#18
     possessed_class=CommonListView
     tracks [{
       type=MovieSceneFloatTrack
       display_name=Render Opacity
       sections [{type=MovieSceneFloatSection, end_frame=10, start_frame=0, channels [None {type=MovieSceneScriptingFloatChannel, keys [[0,1],[10,0]]}]}]
      }]
    }
    ItemDescription {
     id=#19
     possessed_class=WBP_UIKit_DescriptionWithSubtitle_C
     tracks [
      {
       type=MovieSceneFloatTrack
       display_name=Render Opacity
       sections [{type=MovieSceneFloatSection, channels [None {type=MovieSceneScriptingFloatChannel, keys [[0,1],[5,0]]}]}]
      }
      {
       type=MovieScene2DTransformTrack
       sections [{
         type=MovieScene2DTransformSection
         channels [
          Translation.X {type=MovieSceneScriptingFloatChannel, keys [[0,0],[5,128]]}
          Translation.Y {type=MovieSceneScriptingFloatChannel, const [0,2]}
          Angle {type=MovieSceneScriptingFloatChannel}
          Scale.X {type=MovieSceneScriptingFloatChannel}
          Scale.Y {type=MovieSceneScriptingFloatChannel}
          Shear.X {type=MovieSceneScriptingFloatChannel, keys [[0,0],[5,-12]]}
          Shear.Y {type=MovieSceneScriptingFloatChannel, const [0,2]}
         ]
        }]
      }
     ]
    }
    ItemHeaderWidget {
     id=#3
     possessed_class=ItemPresentationGenericRewardHeader_C
     tracks [
      {
       type=MovieScene2DTransformTrack
       sections [{
         type=MovieScene2DTransformSection
         channels [
          Translation.X {type=MovieSceneScriptingFloatChannel, keys [[0.0003,0],[5,-128]]}
          Translation.Y {type=MovieSceneScriptingFloatChannel, const [0,2]}
          Angle {type=MovieSceneScriptingFloatChannel}
          Scale.X {type=MovieSceneScriptingFloatChannel}
          Scale.Y {type=MovieSceneScriptingFloatChannel}
          Shear.X {type=MovieSceneScriptingFloatChannel, keys [[0,0],[5,12]]}
          Shear.Y {type=MovieSceneScriptingFloatChannel, const [0,2]}
         ]
        }]
      }
      {
       type=MovieSceneFloatTrack
       display_name=Render Opacity
       sections [{type=MovieSceneFloatSection, channels [None {type=MovieSceneScriptingFloatChannel, keys [[0.0003,1],[5,0]]}]}]
      }
     ]
    }
    O_Buttons {
     id=#5
     possessed_class=Overlay
     tracks [
      {
       type=MovieSceneFloatTrack
       display_name=Render Opacity
       sections [{type=MovieSceneFloatSection, channels [None {type=MovieSceneScriptingFloatChannel, keys [[5,1],[10,0]]}]}]
      }
      {
       type=MovieScene2DTransformTrack
       sections [{
         type=MovieScene2DTransformSection
         channels [
          Translation.X {type=MovieSceneScriptingFloatChannel}
          Translation.Y {type=MovieSceneScriptingFloatChannel, const [0,2]}
          Angle {type=MovieSceneScriptingFloatChannel}
          Scale.X {type=MovieSceneScriptingFloatChannel}
          Scale.Y {type=MovieSceneScriptingFloatChannel}
          Shear.X {type=MovieSceneScriptingFloatChannel}
          Shear.Y {type=MovieSceneScriptingFloatChannel}
         ]
        }]
      }
     ]
    }
    Button_LeftScrollArrow {
     id=#7
     possessed_class=CircleArrowButton_C
     tracks [
      {
       type=MovieScene2DTransformTrack
       sections [{
         type=MovieScene2DTransformSection
         channels [
          Translation.X {type=MovieSceneScriptingFloatChannel, keys [[2,0],[5,-64]]}
          Translation.Y {type=MovieSceneScriptingFloatChannel}
          Angle {type=MovieSceneScriptingFloatChannel}
          Scale.X {type=MovieSceneScriptingFloatChannel}
          Scale.Y {type=MovieSceneScriptingFloatChannel}
          Shear.X {type=MovieSceneScriptingFloatChannel}
          Shear.Y {type=MovieSceneScriptingFloatChannel}
         ]
        }]
      }
      {
       type=MovieSceneFloatTrack
       display_name=Render Opacity
       sections [{type=MovieSceneFloatSection, channels [None {type=MovieSceneScriptingFloatChannel, keys [[2,1],[5,0]]}]}]
      }
     ]
    }
    Button_RightScrollArrow {
     id=#8
     possessed_class=CircleArrowButton_C
     tracks [
      {
       type=MovieScene2DTransformTrack
       sections [{
         type=MovieScene2DTransformSection
         channels [
          Translation.X {type=MovieSceneScriptingFloatChannel, keys [[2,0],[5,64]]}
          Translation.Y {type=MovieSceneScriptingFloatChannel}
          Angle {type=MovieSceneScriptingFloatChannel}
          Scale.X {type=MovieSceneScriptingFloatChannel}
          Scale.Y {type=MovieSceneScriptingFloatChannel}
          Shear.X {type=MovieSceneScriptingFloatChannel}
          Shear.Y {type=MovieSceneScriptingFloatChannel}
         ]
        }]
      }
      {
       type=MovieSceneFloatTrack
       display_name=Render Opacity
       sections [{type=MovieSceneFloatSection, channels [None {type=MovieSceneScriptingFloatChannel, keys [[2,1],[5,0]]}]}]
      }
     ]
    }
   ]}
 }
]
graphs [
 Activate Gift Box Screen {nodes [
   Activate Gift Box Screen() {is=title K2Node_FunctionEntry}
   n1 = Get GiftBoxGroupVM() -> (GiftBoxGroupVM:Gift Box Group VM Object Reference)
   > n2 = Branch(Condition=Is Not Empty!(TargetArray:Array of Gift Box Item VM Object References=GetAllGiftBoxItems!(self:Gift Box Group VM Object Reference=@n1) -> (ReturnValue:Array of Gift Box Item VM Object References)) -> (ReturnValue:bool))
   > SetActiveWidget(
    self=Get PreMessageSwitcher() -> (PreMessageSwitcher:Common Visibility Switcher Object Reference)
    Widget=Get O_GiftBox() -> (O_GiftBox:Overlay Object Reference)
   )
   > PlayIntro()
   > PlaySound2D(Sound=Get ActivateSound!(self:Gift Box Group VM Object Reference=Get GiftBoxGroupVM()) -> (ActivateSound:Sound Base Object Reference))
   > SetCurrentPreviewItem!(
    execute
    self:Fort Item Preview Widget Object Reference=Get FortItemPreviewWidget() -> (FortItemPreviewWidget:WBP Item Preview Object Reference)
    ItemVM:Fort Item VM Object Reference=Get CurrentItemVM() -> (CurrentItemVM:Fort Item VM Object Reference)
   ) -> (then)
   > SetSelectedIndex!(
    execute
    self:List View Object Reference=Get ListView_Items() -> (ListView_Items:Common List View Object Reference)
    Index:int=Get LastItemIndex() -> (LastItemIndex:int=0)
   ) -> (then)
   > PaginateToNextOrPrevious!(execute, self, IsForward:bool=true) -> (then)
   > SetSidesToPad!(
    execute
    self:Safe Zone Object Reference=Get SZ_GiftBox() -> (SZ_GiftBox:Safe Zone Object Reference)
    InPadLeft:bool=true
    InPadRight:bool=true
    InPadTop:bool=true
    InPadBottom:bool=true
   ) -> (then)
   DeactivateWidget(execute=@n2.else)
  ]}
 AddVisibleItemsToImpression {nodes [
   AddVisibleItemsToImpression() {is=title K2Node_FunctionEntry}
   > n1 = For Each Loop(Array:Array of Fort Item VM Object References=Get ActuallySeenRewardsFromGiftBox() -> (ActuallySeenRewardsFromGiftBox:Array of Fort Item VM Object References)) -> (Array Element:Fort Item VM Object Reference)
   "Fire Event Add Impression (Non-Localized)"!(
    execute=@n1.Completed
    SectionID:String=ItemsSeenInGiftBox
    SectionContents:Array of Flow Analytics Impression Section Content Structures=Get NewImpressions() -> (NewImpressions:Array of Flow Analytics Impression Section Content Structures)
   ) -> (then)
   n2 = GetItemTemplateId!(self:Fort Item VM Object Reference=@n1) -> (ReturnValue:String)
   Add(
    execute=@n1.LoopBody
    TargetArray:Array of Flow Analytics Impression Section Content Structures=Get NewImpressions()
    NewItem:"Flow Analytics Impression Section Content Structure (by ref)"=Make Flow Analytics Impression Section Content!(ContentID:Text="To Text (String)"(InString=@n2), OfferID:Text, Row:int=-1, Column:int=-1, Index:int=-1) -> (FlowAnalyticsImpressionSectionContent:Flow Analytics Impression Section Content Structure) {is=title K2Node_MakeStruct}
   ) {is=title K2Node_CallArrayFunction}
  ]}
 AddVisibleItemsToSeenRewards {nodes [
   AddVisibleItemsToSeenRewards() {is=title K2Node_FunctionEntry}
   > n1 = For Each Loop(Array:Array of User Widget Object References=GetDisplayedEntryWidgets!(self:List View Base Object Reference=Get ListView_Items()) -> (ReturnValue:Array of User Widget Object References)) -> (Array Element:User Widget Object Reference)
   > n2 = Cast To WBP_UIKit_ItemCard_GiftBoxItem!(execute, Object=@n1) -> (then, CastFailed, AsWBP UIKit Item Card Gift Box Item:WBP UIKit Item Card Gift Box Item Object Reference) {is=title K2Node_DynamicCast}
   n3 = Get FortItemVM!(self:WBP UIKit Item Card Gift Box Item Object Reference=@n2) -> (FortItemVM:Fort Item VM Object Reference)
   n4 = Get ActuallySeenRewardsFromGiftBox()
   > n5 = Branch(Condition=Contains Item(TargetArray:Array of Fort Item VM Object References=@n4, ItemToFind:"Fort Item VM Object Reference (by ref)"=@n3))
   Add(execute=@n5.else, TargetArray:Array of Fort Item VM Object References=@n4, NewItem:"Fort Item VM Object Reference (by ref)"=@n3)
  ]}
 BP_GetDesiredFocusTarget {nodes [
   Get Desired Focus Target()
   n1 = Get ListView_Items()
   > Return Node!(
    execute
    ReturnValue:Widget Object Reference="Get (a copy)"(
     Array:Array of User Widget Object References=GetDisplayedEntryWidgets(self=@n1)
     Dimension 1=Get LastSelectedIndex() -> (LastSelectedIndex:int=0)
    ) -> (Output:User Widget Object Reference)
   )
  ]}
 BP_GetDesiredInputConfig {nodes [
   Get Desired Input Config() {is=title K2Node_FunctionEntry}
   > Return Node(ReturnValue:UIInput Config Structure=Make UIInput Config!(
     bIgnoreMoveInput:bool=False
     bIgnoreLookInput:bool=False
     InputMode:ECommonInputMode Enum=All
     MouseCaptureMode:EMouseCaptureMode Enum=CaptureDuringMouseDown
     MouseLockMode:EMouseLockMode Enum=DoNotLock
     bHideCursorDuringViewportCapture:bool=True
    ) -> (UIInputConfig:UIInput Config Structure) {is=title K2Node_MakeStruct})
  ]}
 BP_OnHandleBackAction {nodes [
   On Handle Back Action() {is=title K2Node_FunctionEntry}
   > StartFade!(execute, self, bCloseScreen:bool=true) -> (then)
   > Return Node(ReturnValue:bool=true)
  ]}
 CF_GetButtonsAlignment {nodes [
   n1 = CF_GetButtonsAlignment!() -> (
    then
    ShouldUseOptimizedTouchLayout:bool=~
    StandardAlignment:EHorizontalAlignment Enum=HAlign_Fill
    TouchOptimizedAlignment:EHorizontalAlignment Enum=HAlign_Fill
   ) {is=title K2Node_FunctionEntry}
   > n2 = Branch(Condition=@n1)
   > Return Node!(execute, ReturnedAlignment:EHorizontalAlignment Enum=@n1.TouchOptimizedAlignment)
   Return Node(execute=@n2.else, ReturnedAlignment=@n1.StandardAlignment)
  ]}
 EventGraph {nodes [
   n1 = "On Entry Initialized (ListView_Items)"!() -> (OutputDelegate, then, Item:Object Reference, Widget:User Widget Object Reference) {is=title K2Node_ComponentBoundEvent}
   > On Item Card Initialized!(execute, self, Item VM Object:Object Reference=@n1.Item, Card Widget:User Widget Object Reference=@n1.Widget) -> (then)
   n2 = "On Clicked (WBP_UIKit_Button_GoToLocker)"!() -> (OutputDelegate, then, Button:Common Button Base Object Reference) {is=title K2Node_ComponentBoundEvent}
   n3 = Get ItemName!(self:Item Definition Base Object Reference=Get ItemDefinition!(self:Fort Item VM Object Reference=Get CurrentItemVM()) -> (ItemDefinition:Fort Item Definition Object Reference)) -> (ItemName:Text)
   > Fire Event Element Interacted(
    Element=@n2.Button
    AdditionalInfo="Add String (Non-Localized)"!(
     Info:"Flow Analytics Additional Info Structure (by ref)"=Make Flow Analytics Additional Info!() -> (FlowAnalyticsAdditionalInfo:Flow Analytics Additional Info Structure) {is=title K2Node_MakeStruct}
     Key:String=ItemName
     Value:String="To String (Text)"(InText=@n3)
    ) -> (ReturnValue:Flow Analytics Additional Info Structure)
   )
   > SetIsEnabled(self=Get WBP_UIKit_Button_GoToLocker() -> (WBP_UIKit_Button_GoToLocker:WBP UIKit Button Quiet Object Reference), bInIsEnabled=false)
   > n4 = Set ShouldGoToLocker(ShouldGoToLocker:bool=true) -> (Output_Get:bool)
   PlayAnimation(execute=[@n14.then, @n4.then], InAnimation=Get Anim_Collected() -> (Anim_Collected:Widget Animation Object Reference))
   > PlaySound2D()
   > Delay(Duration=0.75)
   > StartFade(bCloseScreen=false)
   > Set CurrentGiftBoxItem(CurrentGiftBoxItem:Gift Box Item VM Object Reference) -> (Output_Get:Gift Box Item VM Object Reference)
   > Delay(Duration=0.5)
   > PlayAnimationForward(InAnimation=Get Anim_InitialStateSet() -> (Anim_InitialStateSet:Widget Animation Object Reference))
   > n5 = Branch(Condition=Get ShouldGoToLocker() -> (ShouldGoToLocker:bool))
   > NavigateToFeature!(execute, Feature:EFortUIFeature Enum=ShowMyLoadout, bClearPrimaryContentStack:bool=true) -> (then)
   DeactivateWidget(execute=@n5.else)
   n6 = "On Clicked (Button_RightScrollArrow)"!() -> (OutputDelegate, then, Button:Common Button Base Object Reference) {is=title K2Node_ComponentBoundEvent}
   > FireEventElementInteractedByName!(execute, Element Name:String=Button_Next_Page, WidgetCallingFrom:Widget Object Reference=@n6.Button) -> (then, Success:bool)
   > PaginateToNextOrPrevious(IsForward=true)
   "On Entries Generated (ListView_Items)"!() -> (OutputDelegate, then, NumEntries:int=0) {is=title K2Node_ComponentBoundEvent}
   > HandleArrowsVisibility()
   > AddVisibleItemsToSeenRewards() {is=title K2Node_CallFunction}
   n7 = "On Item Clicked (ListView_Items)"!() -> (OutputDelegate, then, Item:Object Reference) {is=title K2Node_ComponentBoundEvent}
   > Select Item!(execute, self, Item:Object Reference=@n7.Item, Active:bool=true) -> (then)
   n8 = "On Clicked (Button_LeftScrollArrow)"!() -> (OutputDelegate, then, Button:Common Button Base Object Reference) {is=title K2Node_ComponentBoundEvent}
   > FireEventElementInteractedByName(Element Name=Button_Previous_Page, WidgetCallingFrom=@n8.Button)
   > PaginateToNextOrPrevious(IsForward=false)
   n9 = "On UIKit Tab Selected (WBP_UIKit_TabGroup)"!() -> (OutputDelegate, then, TabID:Name=None, TabVM:MVVM Base Viewmodel Object Reference) {is=title K2Node_ComponentBoundEvent}
   > n10 = Cast To GiftBoxVM!(execute, Object=@n9.TabVM) -> (then, CastFailed, AsGift Box VM:Gift Box VM Object Reference) {is=title K2Node_DynamicCast}
   > UpdateItemList!(execute, self, GiftBoxVM:Gift Box VM Object Reference=@n10) -> (then)
   n11 = Get WBP_UIKit_TabGroup() -> (WBP_UIKit_TabGroup:WBP UIKit Tab Group Object Reference)
   > Fire Event Element Interacted(Element=GetTabButtonBaseByID!(
     self:Common Tab List Widget Base Object Reference=@n11
     TabNameID:Name=GetActiveTab!(self:Common Tab List Widget Base Object Reference=@n11) -> (ReturnValue:Name=None)
    ) -> (ReturnValue:Common Button Base Object Reference))
   n12 = "On Clicked (Btn_RightAction)"!() -> (OutputDelegate, then, Button:Common Button Base Object Reference) {is=title K2Node_ComponentBoundEvent}
   > FireEventElementInteractedByName(Element Name=Button_Claim_Reward, WidgetCallingFrom=@n12.Button)
   n13 = Get ItemName(self=Get ItemDefinition(self=Get CurrentItemVM()))
   > n14 = Fire Event Element Interacted(
    Element=@n12.Button
    AdditionalInfo="Add String (Non-Localized)"(Info=Make Flow Analytics Additional Info(), Key=ItemName, Value="To String (Text)"(InText=@n13))
   )
   n15 = "On Clicked (Btn_LeftAction)"!() -> (OutputDelegate, then, Button:Common Button Base Object Reference) {is=title K2Node_ComponentBoundEvent}
   n16 = Get ItemVM!(self:Gift Box Item VM Object Reference=Get CurrentGiftBoxItem() -> (CurrentGiftBoxItem:Gift Box Item VM Object Reference)) -> (ItemVM:Fort Item VM Object Reference)
   > n17 = Branch(Condition=RequiresCustomization!(self:Fort Item VM Object Reference=@n16) -> (ReturnValue:bool))
   > OpenCustomizeMenu!(execute, self, ItemVM:Fort Item VM Object Reference=@n16) -> (then)
   > FireEventElementInteractedByName(Element Name=Button_Customize, WidgetCallingFrom=@n15.Button)
   n18 = Get ItemName(self=Get ItemDefinition(self=Get CurrentItemVM()))
   > Fire Event Element Interacted(
    Element=@n15.Button
    AdditionalInfo="Add String (Non-Localized)"(Info=Make Flow Analytics Additional Info(), Key=ItemName, Value="To String (Text)"(InText=@n18))
   )
   EquipItem!(execute=@n17.else, self, ItemVM:Fort Item VM Object Reference=@n16) -> (then)
   > FireEventElementInteractedByName(Element Name=Button_Equip, WidgetCallingFrom=@n15.Button)
   n19 = Get ItemName(self=Get ItemDefinition(self=Get CurrentItemVM()))
   > Fire Event Element Interacted(
    Element=@n15.Button
    AdditionalInfo="Add String (Non-Localized)"(Info=Make Flow Analytics Additional Info(), Key=ItemName, Value="To String (Text)"(InText=@n19))
   )
   n20 = "On Item Selection Changed (ListView_Items)"!() -> (OutputDelegate, then, Item:Object Reference, bIsSelected:bool) {is=title K2Node_ComponentBoundEvent}
   > Select Item(Item=@n20.Item, Active=@n20.bIsSelected)
   n21 = Set Left Side Content!() -> (OutputDelegate, then, Custom Widget:User Widget Soft Class Reference) {is=title K2Node_CustomEvent}
   > n22 = Branch(Condition=IsValidSoftClassReference!(SoftClassReference:"Object Soft Class Reference (by ref)"=@n21.Custom Widget) -> (ReturnValue:bool))
   > SetActiveWidget(
    self:Widget Switcher Object Reference=Get LeftsideContentSwitcher() -> (LeftsideContentSwitcher:Widget Switcher Object Reference)
    Widget=Get Overlay_CustomContent() -> (Overlay_CustomContent:Overlay Object Reference)
   )
   > n23 = CreateWidgetAsync!(
    execute:Exec
    UserWidgetSoftClass:User Widget Soft Class Reference=@n21.Custom Widget
    OwningPlayer:Player Controller Object Reference=GetOwningPlayer()
    bSuspendInputUntilComplete:bool=false
   ) -> (then:Exec, OnComplete:Exec, OnFailure:Exec, UserWidget:User Widget Object Reference) {is=title K2Node_AsyncAction}
   Set Custom Header Widget!(execute=@n23.OnComplete, self, NewWidget:User Widget Object Reference=@n23) -> (then) {is=title K2Node_CallFunction}
   > Process Background Transition()
   SetActiveWidget(execute=@n22.else, self=Get LeftsideContentSwitcher(), Widget=Get DefaultContent() -> (DefaultContent:Overlay Object Reference))
   > RequestLoadBackground!(
    execute
    self
    Background Asset:Soft Object Reference
    Direction:EVaultWorldTransitionDirection Enum=Get BackgroundTransitionDirection() -> (BackgroundTransitionDirection:EVaultWorldTransitionDirection Enum=Forward)
   ) -> (then)
   n24 = HandleInputMethodChanged!() -> (OutputDelegate, then, NewInputType:ECommonInputType Enum) {is=title K2Node_CustomEvent}
   > Set Input Type(Input Type:ECommonInputType Enum=@n24.NewInputType) -> (Output_Get:ECommonInputType Enum)
   > HandleArrowsVisibility()
   > SetMobileStyle() {is=title K2Node_CallFunction}
   n25 = On Premessage Deactivate() {is=title K2Node_CustomEvent}
   Activate Gift Box Screen(execute=[@n25.then, @n50.CastFailed]) {is=title K2Node_CallFunction}
   n26 = RequestLoadBackground!() -> (OutputDelegate, then, Background Asset:Soft Object Reference, Direction:EVaultWorldTransitionDirection Enum=Forward) {is=title K2Node_CustomEvent}
   > Set PendingTransitionDirection(PendingTransitionDirection:EVaultWorldTransitionDirection Enum=@n26.Direction) -> (Output_Get:EVaultWorldTransitionDirection Enum=Forward)
   > Set PendingBackgroundSoftPath(PendingBackgroundSoftPath:Soft Object Path Structure=To Soft Object Path!(SoftObjectReference:Soft Object Reference=@n26.Background Asset) -> (ReturnValue:Soft Object Path Structure)) -> (Output_Get:Soft Object Path Structure)
   > n27 = Async Load Asset(Asset=@n26.Background Asset)
   Branch(
    execute=@n27.Completed
    Condition="Equal Exactly (String)"(A=GetSoftObjectPath!(Object=@n27) -> (ReturnValue_PathString:String), B=Get!() -> (PendingBackgroundSoftPath_PathString:String))
   )
   > Branch(Condition=Get bIsActive() -> (bIsActive:bool))
   > UpdateBackground!(
    execute
    self:Fort Item Preview Background Object Reference=Get FortItemPreviewBackground() -> (FortItemPreviewBackground:Fort Item Preview Background Object Reference)
    LoadedBackground:Object Reference=@n27
    TransitionDirection:EVaultWorldTransitionDirection Enum=Get PendingTransitionDirection() -> (PendingTransitionDirection:EVaultWorldTransitionDirection Enum=Forward)
    PrimaryColor:Linear Color Structure=(R=0,G=0,B=0,A=0)
    SecondaryColor:Linear Color Structure=(R=0,G=0,B=0,A=0)
   ) -> (then)
   > SetVectorParameterValue!(
    execute
    Collection:Material Parameter Collection Object Reference=/Game/Environments/FrontEnd/Blueprints/ItemPreview/VaultWorldTransitions/MaterialCollectionParameters/MPC_UI_Master_VaultWorld_Settings
    ParameterName:Name=FocalPointPosition
    ParameterValue_R:float=0
    ParameterValue_G:float=0.1
    ParameterValue_B:float=0
    ParameterValue_A:float=1
   ) -> (then) {is=title K2Node_CallMaterialParameterCollectionFunction}
   Process Background Transition() {is=title K2Node_CustomEvent}
   > n28 = Sequence()
   > SetScalarParameterValue!(
    execute
    Collection:Material Parameter Collection Object Reference=&MPC_UI_Master_VaultWorld_Settings
    ParameterName:Name=Is Horizontal Orientation
    ParameterValue:float=1
   ) -> (then) {is=title K2Node_CallMaterialParameterCollectionFunction}
   > SetVectorParameterValue(
    Collection=&MPC_UI_Master_VaultWorld_Settings
    ParameterName=FocalPointPosition
    ParameterValue_R=0
    ParameterValue_G=0.1
    ParameterValue_B=0
    ParameterValue_A=1
   )
   n29 = Get LeftsideContentSwitcher()
   n30 = Branch(
    execute=@n28.then_1
    Condition="Equal (Object)"(A=Get Overlay_CustomContent(), B=GetActiveWidget!(self:Widget Switcher Object Reference=@n29) -> (ReturnValue:Widget Object Reference))
   )
   > n31 = GetCustomBackground!(
    execute:Exec
    self:Object Reference=GetContent!(self:Content Widget Object Reference=Get CustomContent() -> (CustomContent:Named Slot Object Reference)) -> (ReturnValue:Widget Object Reference)
   ) -> (then:Exec, Background:Soft Object Reference) {is=title K2Node_Message}
   > RequestLoadBackground(Background Asset=@n31, Direction=Get BackgroundTransitionDirection() -> (BackgroundTransitionDirection=Forward)) {is=title K2Node_CallFunction}
   RequestLoadBackground(execute=@n30.else, Direction=Get BackgroundTransitionDirection() -> (BackgroundTransitionDirection=Forward))
   n32 = SetRightSideWidget_Event!() -> (OutputDelegate, then, Custom Widget:User Widget Soft Class Reference) {is=title K2Node_CustomEvent}
   > n33 = Branch(Condition=IsValidSoftClassReference(SoftClassReference=@n32.Custom Widget))
   > n34 = CreateWidgetAsync(UserWidgetSoftClass=@n32.Custom Widget, OwningPlayer=GetOwningPlayer(), bSuspendInputUntilComplete=false)
   SetContent!(
    execute=@n34.OnComplete
    self:Content Widget Object Reference=Get Slot_GiftboxCustomBanner() -> (Slot_GiftboxCustomBanner:Named Slot Object Reference)
    Content:Widget Object Reference=@n34
   ) -> (then, ReturnValue:Panel Slot Object Reference)
   ClearChildren(execute=@n33.else, self=Get Slot_GiftboxCustomBanner())
   n35 = OpenCustomizeMenu!() -> (OutputDelegate, then, ItemVM:Fort Item VM Object Reference) {is=title K2Node_CustomEvent}
   > Set ItemToCustomize(ItemToCustomize:Fort Item VM Object Reference=@n35.ItemVM) -> (Output_Get:Fort Item VM Object Reference)
   n36 = TryGetSoftCustomActivatableWidgetClass!(
    self:Fort Locker Category Config Object Reference=GetItemCategory!(self:Fort Item VM Object Reference=@n35.ItemVM) -> (ReturnValue:Fort Locker Category Config Object Reference)
    ViewRequest:Gameplay Tag Structure=(TagName="UI.ItemPreview.PluginBridge.Mimosa.CustomizeMenu")
   ) -> (WidgetClass:Common Activatable Widget Soft Class Reference, ReturnValue:bool)
   > Branch(Condition=@n36.ReturnValue)
   > n37 = PushPrimaryContentSoftClassAsync!(execute:Exec, SoftWidgetClass:Common Activatable Widget Soft Class Reference=@n36) -> (then:Exec, OnPrimaryContentPushed:Exec, ActivatableWidget:Common Activatable Widget Object Reference) {is=title K2Node_AsyncAction}
   SetViewModelByClass!(
    execute=@n37.OnPrimaryContentPushed
    Widget:User Widget Object Reference=@n37
    ViewModel:Notify Field Value Changed Interface=Get ItemToCustomize() -> (ItemToCustomize:Fort Item VM Object Reference)
   ) -> (then)
   n38 = EquipItem!() -> (OutputDelegate, then, ItemVM:Fort Item VM Object Reference) {is=title K2Node_CustomEvent}
   > n39 = ShowConfirmationOnEquipIfRequired!(
    execute:Exec
    Item:Fort Item VM Object Reference=@n38.ItemVM
    LockerConfirmationModalClass:Locker Confirmation Modal Soft Class Reference=/Game/Athena/UI/Customization/LockerArchivingModal.LockerArchivingModal_C
   ) -> (then:Exec, OnOperationCompleted:Exec, Success:bool=~) {is=title K2Node_AsyncAction}
   > n40 = Branch(Condition=@n39)
   n41 = EquipItem!(
    execute:Exec=[@n39.OnOperationCompleted, @n40.then]
    Item:Fort Item VM Object Reference=@n38.ItemVM
    SlotIndex:int=0
    Variants:Array of Mcp Variant Channel Info Structures
    bApplyToAllSlots:bool=false
    bAllowMocking:bool=true
   ) -> (then:Exec, OnOperationCompleted:Exec, Success:bool=~) {is=title K2Node_AsyncAction}
   n42 = Get Btn_LeftAction() -> (Btn_LeftAction:WBP UIKit Button Regular Object Reference)
   > SetIsInteractionEnabled!(execute, self:Common Button Base Object Reference=@n42, bInIsInteractionEnabled:bool=false) -> (then)
   > SetIsEnabled(self=@n42, bInIsEnabled=false)
   PlayAnimation(execute=@n41.OnOperationCompleted, InAnimation=Get Anim_Equipped() -> (Anim_Equipped:Widget Animation Object Reference))
   > PlaySound2D()
   Event On Activated()
   > ActivateWidget(self=Get FortItemPreviewWidget())
   > HandleAnimationInitialState()
   > Set ShouldGoToLocker(ShouldGoToLocker=false)
   > Clear(TargetArray:Array of Fort Item VM Object References=Get ActuallySeenRewardsFromGiftBox())
   n43 = Get WBP_UIKit_Button_GoToLocker()
   > SetGridSlotAlignment!(
    execute
    Widget:Widget Object Reference=@n43
    Horizontal:EHorizontalAlignment Enum=HAlign_Left
    Vertical:EVerticalAlignment Enum=VAlign_Center
   ) -> (then)
   > SetIsEnabled(self=@n43, bInIsEnabled=true)
   > n44 = Sequence!(execute) -> (then_0, then_1, then_2:Exec, then_3:Exec)
   > n45 = GetConsoleVariableBoolValue!(execute, VariableName:String=Fort.Rollback.GBsSetFrontendCameraBeforeActivating) -> (then, ReturnValue:bool)
   > Branch(Condition=@n45)
   > UpdateFrontendCamera()
   n46 = Get PreMessageScreen!(self:Gift Box Group VM Object Reference=Get GiftBoxGroupVM()) -> (PreMessageScreen:Common Activatable Widget Soft Class Reference)
   n47 = Get PreMessagePushed() -> (PreMessagePushed:bool)
   n48 = Branch(execute=@n44.then_1, Condition=AND Boolean(A=NOT Boolean(A=@n47), B=IsValidSoftClassReference(SoftClassReference=@n46)))
   > Set PreMessagePushed(PreMessagePushed:bool=true) -> (Output_Get:bool)
   > n49 = CreateWidgetAsync(UserWidgetSoftClass=Get PreMessageScreen(self=Get GiftBoxGroupVM()), bSuspendInputUntilComplete=false)
   n50 = Cast To CommonActivatableWidget!(execute=@n49.OnComplete, Object=@n49) -> (then, CastFailed, AsCommon Activatable Widget:Common Activatable Widget Object Reference) {is=title K2Node_DynamicCast}
   > Bind Event to On Widget Deactivated(self:Common Activatable Widget Object Reference=@n50, Delegate=@n25) {is=title K2Node_AddDelegate}
   > SetContent(self=Get PreMessageSlot() -> (PreMessageSlot:Named Slot Object Reference), Content=@n50)
   n51 = GetExtension!(self:User Widget Object Reference=@n50, ExtensionType:User Widget Extension Class Reference=/Script/ModelViewViewModel.MVVMView) -> (ReturnValue:MVVM View Object Reference)
   > n52 = Is Valid(InputObject=@n51)
   SetActiveWidgetIndex(execute=[@n54.then, @n52.Is Not Valid], self:Common Visibility Switcher Object Reference=Get PreMessageSwitcher(), Index=1)
   > ActivateWidget(self=@n50)
   > SetCurrentPreviewItem(self=Get FortItemPreviewWidget())
   n53 = Get GiftBoxGroupVM()
   n54 = SetViewModelByClass!(
    execute=@n52.Is Valid
    self:MVVM View Object Reference=@n51
    NewValue:Notify Field Value Changed Interface="Get (a copy)"(Array:Array of Gift Box VM Object References=Get GiftBoxes!(self:Gift Box Group VM Object Reference=@n53) -> (GiftBoxes:Array of Gift Box VM Object References)) -> (Output:Gift Box VM Object Reference)
   ) -> (then, ReturnValue:bool)
   Activate Gift Box Screen(execute=@n48.else)
   SetMobileStyle(execute=@n44.then_3)
   SetVisibility(
    execute=@n44.then_2
    self=Get WBP_UIKit_Button_GoToLocker()
    InVisibility="To Visibility (Boolean)"!(
     bIsVisible:bool=GetConfigDataBoolWithDefault!(EntryName:String=EnableGoToLockerInGiftbox, bDefaultValue:bool=true) -> (ReturnValue:bool)
     TrueVisibility:ESlateVisibility Enum=SelfHitTestInvisible
     FalseVisibility:ESlateVisibility Enum=Collapsed
    ) -> (ReturnValue:ESlateVisibility Enum=Visible)
   )
   > n55 = GetConsoleVariableBoolValue(VariableName=Fort.Rollback.GBsSetFrontendCameraBeforeActivating)
   > n56 = Branch(Condition=@n55)
   UpdateFrontendCamera(execute=@n56.else)
   Event On Deactivated()
   n57 = Get FortItemPreviewWidget()
   > SetCurrentPreviewItem(self=@n57)
   > DeactivateWidget(self=@n57)
   > AddVisibleItemsToImpression() {is=title K2Node_CallFunction}
   > Set CurrentItemVM(CurrentItemVM:Fort Item VM Object Reference) -> (Output_Get:Fort Item VM Object Reference)
   > Set CurrentGiftBoxItem()
   > SetSidesToPad(self=Get SZ_GiftBox(), InPadLeft=false, InPadRight=false, InPadTop=false, InPadBottom=false)
   Event Construct()
   n58 = Create Event()
   n59 = Get CommonInputSubsystem()
   > Bind Event to On Input Method Changed(self=@n59, Delegate=@n58)
   > HandleInputMethodChanged!(execute, self, NewInputType:ECommonInputType Enum=GetCurrentInputType(self=@n59)) -> (then) {is=title K2Node_CallFunction}
   > SetMaxDesiredHeight!(
    execute
    self:Size Box Object Reference=Get SizeBoxTitle!(self:WBP UIKit Item Description Base Object Reference=Get ItemDescription() -> (ItemDescription:WBP UIKit Description With Subtitle Object Reference)) -> (SizeBoxTitle:Size Box Object Reference)
    InMaxDesiredHeight:float=160
   ) -> (then)
   > SetMobileStyle()
   > Set PreMessagePushed(PreMessagePushed=false)
   Event Destruct()
   > Unbind Event from On Input Method Changed(self=@n59, Delegate=@n58)
   > ClearChildren(self=Get Slot_GiftboxCustomBanner())
   K2Node_Event_8 = OnUpdateItemContextAction_1!() -> (OutputDelegate, then, ItemContextAction:Byte) {is=title K2Node_CustomEvent}
   > Set Text(
    Text=Select!(Option 0:Text, Option 1:Text, Option 2:Text, Option 3:Text, Index:Byte=@K2Node_Event_8.ItemContextAction) -> (ReturnValue:Text)
    self:Fort CTAButton Object Reference=Get Btn_LeftAction()
   )
  ]}
 Get Collected Items Text {nodes [
   n1 = Get Collected Items Text!() -> (then, Gift Box Group:Gift Box Group VM Object Reference) {is=title K2Node_FunctionEntry}
   > n2 = For Each Loop(Array:Array of Gift Box VM Object References=Get GiftBoxes(self=@n1)) -> (Array Element:Gift Box VM Object Reference)
   Return Node!(
    execute=@n2.Completed
    Result:Text=Format Text!(Format="LOCTEXT(\"{0} {0}|plural(one=Item,other=Items) Collected\")", 0:int=Get Count() -> (Count:int=0)) -> (Result)
   )
   n3 = Get Items!(self:Gift Box VM Object Reference=@n2) -> (Items:Array of Gift Box Item VM Object References)
   > Set Count(Count:int=int + int(A=Length(TargetArray:Array of Gift Box Item VM Object References=@n3), B=Get Count() -> (Count=0))) -> (Output_Get:int=0)
  ]}
 GetLeftActionVisibility {nodes [
   GetLeftActionVisibility!() -> (then, Can Equip:bool=~, Is Equipped:bool=~, RequiresCustomization:bool=~, IsBulkCosmeticInstalled:bool=~) {is=title K2Node_FunctionEntry}
   > n1 = Branch(Condition=Get RequiresCustomization() -> (RequiresCustomization:bool))
   > n2 = Branch(Condition=GetIsBulkCosmeticInstalled!(self:Fort Download System VM Object Reference=Get FortDownloadSystemVM() -> (FortDownloadSystemVM:Fort Download System VM Object Reference)) -> (ReturnValue:bool))
   n3 = Get Is Equipped() -> (Is Equipped:bool)
   Return Node(execute=[@n1.else, @n2.then], Result:bool=AND Boolean(A=Get Can Equip() -> (Can Equip:bool), B=NOT Boolean(A=@n3)))
   Return Node(execute=@n2.else, Result=false)
  ]}
 GetPrimaryContentSetup {nodes [
   GetPrimaryContentSetup() {is=title K2Node_FunctionEntry}
   > Return Node!(
    execute
    ReturnValue:Primary Content Setup Structure=Make Primary Content Setup!(
     bShowBottomBar:bool=false
     bShowChatWidget:bool=false
     bEnableNewTextChat:bool=False
     bShowTopBar:bool=false
     bTabSupportsInterface:bool=False
    ) -> (PrimaryContentSetup:Primary Content Setup Structure) {is=title K2Node_MakeStruct}
   )
  ]}
 HandleAnimationInitialState {nodes [
   HandleAnimationInitialState() {is=title K2Node_FunctionEntry}
   > PlayAnimation(InAnimation=Get Anim_InitialStateSet())
  ]}
 HandleArrowsVisibility {nodes [
   HandleArrowsVisibility() {is=title K2Node_FunctionEntry}
   n1 = Get ListView_Items()
   n2 = GetItemAt!(self:List View Object Reference=@n1, Index:int=0) -> (ReturnValue:Object Reference)
   n3 = Get Input Type() -> (Input Type:ECommonInputType Enum)
   > SetVisibility(
    self=Get Button_LeftScrollArrow() -> (Button_LeftScrollArrow:WBP UIKit Button Regular Object Reference)
    InVisibility=Select!(
     Option 0:ESlateVisibility Enum=Select!(
      MouseAndKeyboard:ESlateVisibility Enum=Visible
      Gamepad:ESlateVisibility Enum=Hidden
      Touch:ESlateVisibility Enum=Visible
      Count:ESlateVisibility Enum=Visible
      Index:ECommonInputType Enum=@n3
     ) -> (ReturnValue)
     Option 1:ESlateVisibility Enum=Hidden
     Index:bool=Is Item Visible!(self:List View Object Reference=@n1, Item:Object Reference=@n2) -> (ReturnValue:bool)
    ) -> (ReturnValue:ESlateVisibility Enum)
   )
   n4 = Get ListView_Items()
   n5 = GetItemAt(self=@n4, Index=int - int(A=GetNumItems!(self:List View Object Reference=@n4) -> (ReturnValue:int=0)))
   n6 = Get Input Type()
   > SetVisibility(
    self=Get Button_RightScrollArrow() -> (Button_RightScrollArrow:WBP UIKit Button Regular Object Reference)
    InVisibility=Select!(
     Option 0:ESlateVisibility Enum=Select!(
      MouseAndKeyboard:ESlateVisibility Enum=Visible
      Gamepad:ESlateVisibility Enum=Hidden
      Touch:ESlateVisibility Enum=Visible
      Count:ESlateVisibility Enum=Visible
      Index:ECommonInputType Enum=@n6
     ) -> (ReturnValue)
     Option 1:ESlateVisibility Enum=Hidden
     Index:bool=Is Item Visible(self=@n4, Item=@n5)
    ) -> (ReturnValue)
   )
  ]}
 HandleFadeAnimFinished {nodes [
   HandleFadeAnimFinished() {is=title K2Node_FunctionEntry}
   > UnbindFromAnimationFinished!(
    execute
    self:User Widget Object Reference
    Animation:Widget Animation Object Reference=Get Anim_Fade() -> (Anim_Fade:Widget Animation Object Reference)
    Delegate:Delegate=Create Event()
   ) -> (then)
   > SetVisibility(self=Get ItemHeaderWidget() -> (ItemHeaderWidget:Gift Box Screen Header Object Reference), InVisibility=Visible)
   > SetActiveWidget(self=Get LeftsideContentSwitcher(), Widget=Get DefaultContent())
   > DeactivateWidget()
  ]}
 HandleItemListVisible {nodes [
   HandleItemListVisible() {is=title K2Node_FunctionEntry}
   n1 = Get ListView_Items()
   > Set HasOnlyOneItem(HasOnlyOneItem:bool="integer <= integer"(A=GetNumItems(self=@n1) -> (ReturnValue=0), B=Get ItemMinLimit() -> (ItemMinLimit:int=0))) -> (Output_Get:bool)
   n2 = Get HasOnlyOneItem() -> (HasOnlyOneItem:bool)
   n3 = Get GP_ItemList() -> (GP_ItemList:Grid Panel Object Reference)
   > SetRenderOpacity(self=@n3, InOpacity=Select!(Option 0:float=1, Option 1:float=0, Index:bool=@n2) -> (ReturnValue:float))
   > SetVisibility(
    self=@n3
    InVisibility=Select(Option 0:ESlateVisibility Enum=Visible, Option 1:ESlateVisibility Enum=HitTestInvisible, Index=@n2) -> (ReturnValue:ESlateVisibility Enum)
   )
   > SetPadding!(
    execute
    self:Vertical Box Slot Object Reference=SlotAsVerticalBoxSlot(Widget=Get O_Buttons() -> (O_Buttons:Overlay Object Reference))
    InPadding_Left:float=80
    InPadding_Top:float=Select(Option 0:float=32, Option 1:float=0, Index=@n2) -> (ReturnValue:float)
    InPadding_Right:float=80
    InPadding_Bottom:float=0
   ) -> (then)
  ]}
 HandleItemsFadeVisibility {nodes [
   HandleItemsFadeVisibility() {is=title K2Node_FunctionEntry}
   > HandleArrowsVisibility() {is=title K2Node_CallFunction}
  ]}
 MergeGiftBoxesIntoOne {nodes [
   n1 = MergeGiftBoxesIntoOne!() -> (then, InGiftBoxes:Array of Gift Box VM Object References, ShouldMerge:bool=~) {is=title K2Node_FunctionEntry}
   > n2 = Branch(Condition=@n1.ShouldMerge)
   > n3 = Construct Object from Class(Class=/Script/GiftBox.GiftBoxVM) -> (ReturnValue:Gift Box VM Object Reference)
   > Set SingleGiftBoxVM(SingleGiftBoxVM:Gift Box VM Object Reference=@n3) -> (Output_Get:Gift Box VM Object Reference)
   > n4 = For Each Loop(Array=Get InGiftBoxes() -> (InGiftBoxes:Array of Gift Box VM Object References))
   Return Node!(
    execute=@n4.Completed
    OutGiftBoxes:Array of Gift Box VM Object References=Make Array("[0]":Gift Box VM Object Reference=Get SingleGiftBoxVM() -> (SingleGiftBoxVM:Gift Box VM Object Reference)) -> (Array:Array of Gift Box VM Object References)
   )
   > n5 = For Each Loop(Array:Array of Gift Box Item VM Object References=Get Items(self=@n4)) -> (Array Element:Gift Box Item VM Object Reference)
   > Add(
    TargetArray:Array of Gift Box Item VM Object References=Get Items(self=Get SingleGiftBoxVM())
    NewItem:"Gift Box Item VM Object Reference (by ref)"=@n5
   )
   Return Node(execute=@n2.else, OutGiftBoxes=Get InGiftBoxes())
  ]}
 On Item Card Initialized {nodes [
   n1 = On Item Card Initialized!() -> (then, Item VM Object:Object Reference, Card Widget:User Widget Object Reference) {is=title K2Node_FunctionEntry}
   > n2 = Cast To WBP_UIKit_ItemCard_GiftBoxItem(Object=@n1.Card Widget)
   > n3 = Cast To GiftBoxItemVM!(execute, Object=Get Item VM Object() -> (Item VM Object:Object Reference)) -> (then, CastFailed, AsGift Box Item VM:Gift Box Item VM Object Reference) {is=title K2Node_DynamicCast}
   > SetGiftBoxItemVM!(execute, self:WBP UIKit Item Card Gift Box Item Object Reference=@n2, Viewmodel:Gift Box Item VM Object Reference=@n3) -> (then)
   LogString!(
    execute=@n2.CastFailed
    Instance Log Level:BlueprintLogLevel Enum=NewEnumerator5
    LogString:String=Entry type of the GiftBoxScreen is not WBP_UIKit_ItemCard_GiftBoxItem. Please adjust the code to correspond to current widget class
    PrintToScreen:bool=true
    Instigator:Object Reference=Self-Reference()
   ) -> (then)
  ]}
 OnGiftBoxesSet {nodes [
   OnGiftBoxesSet!() -> (then, Gift Boxes:Array of Gift Box VM Object References) {is=title K2Node_FunctionEntry}
   > Set PreMessagePushed(PreMessagePushed=false)
   > RemoveTabs(self:WBP UIKit Tab Group Base Object Reference=Get WBP_UIKit_TabGroup())
   n1 = Get Gift Boxes() -> (Gift Boxes:Array of Gift Box VM Object References)
   > n2 = MergeGiftBoxesIntoOne!(
    execute
    self
    InGiftBoxes:Array of Gift Box VM Object References=@n1
    ShouldMerge:bool=ShouldMergeRewardsIntoSingleGiftBox!(GiftBoxes:Array of Gift Box VM Object References=@n1) -> (IsAllInOneGiftBox:bool)
   ) -> (then, OutGiftBoxes:Array of Gift Box VM Object References) {is=title K2Node_CallFunction}
   > n3 = For Each Loop(Array:Array of Gift Box VM Object References=@n2) -> (Array Element:Gift Box VM Object Reference)
   n4 = Get WBP_UIKit_TabGroup()
   SelectTabByIndex!(execute=@n3.Completed, self:UIKit Tab Group Object Reference=@n4, Index:int=0, bSuppressClickFeedback:bool=true) -> (then, ReturnValue:bool)
   n5 = Length(TargetArray:Array of Gift Box VM Object References=@n2)
   > SetVisibility(
    self=@n4
    InVisibility=Select(Option 0:ESlateVisibility Enum=SelfHitTestInvisible, Option 1:ESlateVisibility Enum=Collapsed, Index="integer <= integer"(A=@n5, B=1)) -> (ReturnValue:ESlateVisibility Enum)
   )
   n6 = Get TabText!(self:Gift Box VM Object Reference=@n3) -> (TabText:Text)
   AddTab!(
    execute=@n3.LoopBody
    self:WBP UIKit Tab Group Base Object Reference=Get WBP_UIKit_TabGroup()
    Tab VM:MVVM Base Viewmodel Object Reference=@n3
    Tab ID:Name=String To Name(InString="To String (Text)"(InText=@n6))
    Index:int=-1
   ) -> (then)
  ]}
 PaginateToNextOrPrevious {nodes [
   PaginateToNextOrPrevious!() -> (then, IsForward:bool=~) {is=title K2Node_FunctionEntry}
   > n1 = Sequence!(execute) -> (then_0, then_1, then_2)
   n2 = Get ListView_Items()
   > Set SelectedItemIndex(SelectedItemIndex:int=GetIndexForItem!(
     self:List View Object Reference=@n2
     Item:Object Reference=Get Selected Item!(self:List View Object Reference=@n2) -> (ReturnValue:Object Reference) {is=title K2Node_CallFunction}
    ) -> (ReturnValue:int=0)) -> (Output_Get:int=0)
   n3 = Get ListView_Items()
   > Set CountOfItemsDisplayed(CountOfItemsDisplayed:int=Length(TargetArray:Array of User Widget Object References=GetDisplayedEntryWidgets(self=@n3))) -> (Output_Get:int=0)
   > Set TotalCountOfItems(TotalCountOfItems:int=GetNumItems(self=Get ListView_Items()) -> (ReturnValue=0)) -> (Output_Get:int=0)
   > Set SelectedItemPage(SelectedItemPage:int=int / int(A=Get SelectedItemIndex() -> (SelectedItemIndex:int=0), B=Get CountOfItemsDisplayed() -> (CountOfItemsDisplayed:int=0))) -> (Output_Get:int=0)
   n4 = "Not Equal (Integer)"(A=Get LastSelectedIndex() -> (LastSelectedIndex=0), B=Get SelectedItemIndex() -> (SelectedItemIndex=0))
   n5 = SelectInt(A=0, B=Get Pages Moved() -> (Pages Moved:int=0), bPickA=@n4)
   n6 = SelectInt(A=1, B=-1, bPickA=Get IsForward() -> (IsForward:bool))
   n7 = int / int(A=Get TotalCountOfItems() -> (TotalCountOfItems:int=0), B=Get CountOfItemsDisplayed() -> (CountOfItemsDisplayed=0))
   n8 = Get SelectedItemPage() -> (SelectedItemPage:int=0)
   n9 = Get SelectedItemPage() -> (SelectedItemPage=0)
   Set Pages Moved(
    execute=@n1.then_1
    Pages Moved:int="Clamp (Integer)"(Value=int + int(A=@n5, B=@n6), Min=int * int(A=@n9, B=-1), Max=int - int(A=@n7, B=@n8)) -> (ReturnValue=0)
   ) -> (Output_Get:int=0)
   > Branch(Condition=@n4)
   > Set LastSelectedIndex(LastSelectedIndex:int=Get SelectedItemIndex() -> (SelectedItemIndex=0)) -> (Output_Get:int=0)
   n10 = int - int(A=Get TotalCountOfItems() -> (TotalCountOfItems=0))
   n11 = Get SelectedItemIndex() -> (SelectedItemIndex=0)
   n12 = Get CountOfItemsDisplayed() -> (CountOfItemsDisplayed=0)
   n13 = int * int(A=int + int(A=int / int(A=@n11, B=@n12), B=Get Pages Moved() -> (Pages Moved=0)), B=@n12)
   n14 = int + int!(A=@n13, B=int / int(A=Get CountOfItemsDisplayed() -> (CountOfItemsDisplayed=0), B=2), C:int=-1) -> (ReturnValue)
   n15 = int - int(A=Get TotalCountOfItems() -> (TotalCountOfItems=0))
   NavigateToIndex!(
    execute=@n1.then_2
    self:List View Object Reference=Get ListView_Items()
    Index:int=SelectInt(
     A="Clamp (Integer)"(Value=@n14, Min=0, Max=@n15) -> (ReturnValue=0)
     B="Clamp (Integer)"(Value=@n13, Min=0, Max=@n10) -> (ReturnValue=0)
     bPickA=Get IsForward()
    )
   ) -> (then)
   SelectInt(A=1, B=-1, bPickA=Get IsForward())
  ]}
 PlayIntro {nodes [
   PlayIntro() {is=title K2Node_FunctionEntry}
   > PlayAnimation(InAnimation=Get Anim_Intro() -> (Anim_Intro:Widget Animation Object Reference))
   > SpawnSound2D!(
    execute
    Sound:Sound Base Object Reference=Get WhooshSound() -> (WhooshSound:Sound Base Object Reference)
    VolumeMultiplier:float=1
    PitchMultiplier:float=1
    StartTime:float=0
    ConcurrencySettings:Sound Concurrency Object Reference
    bPersistAcrossLevelTransition:bool=false
    bAutoDestroy:bool=true
   ) -> (then, ReturnValue:Audio Component Object Reference)
  ]}
 Select Item {nodes [
   n1 = Select Item!() -> (then, Item:Object Reference, Active:bool=true) {is=title K2Node_FunctionEntry}
   > Branch(Condition=@n1.Active)
   > Set Local Current Index(Local Current Index:int=GetIndexForItem(self=Get ListView_Items(), Item=Get Item() -> (Item:Object Reference)) -> (ReturnValue=0)) -> (Output_Get:int=0)
   > n2 = Cast To GiftBoxItemVM(Object=@n1)
   > n3 = Set CurrentGiftBoxItem(CurrentGiftBoxItem=@n2)
   > n4 = Cast To FortPlayerController(Object=GetOwningPlayer())
   n5 = Get ItemName(self=Get ItemDefinition(self=Get CurrentItemVM()))
   > "Fire Event Element Interacted (Non-Localized)"!(
    execute
    FortPC:Fort Player Controller Object Reference=@n4
    ElementName:String=GetItemTemplateId(self=Get ItemVM(self=@n3))
    AdditionalInfo:Flow Analytics Additional Info Structure="Add String (Non-Localized)"(Info=Make Flow Analytics Additional Info(), Key=ItemName, Value="To String (Text)"(InText=@n5))
   ) -> (then)
   > SetCompatibilityVisibility!(execute, self:WBP UIKit Item Description Base Object Reference=Get ItemDescription(), Show:bool=false) -> (then, IsVisible:bool)
   n6 = Get LastItemIndex() -> (LastItemIndex=0)
   n7 = Get Local Current Index() -> (Local Current Index:int=0)
   > Branch(Condition="Not Equal (Integer)"(A=@n7, B=@n6))
   > Set BackgroundTransitionDirection(BackgroundTransitionDirection:EVaultWorldTransitionDirection Enum=Select(
     Option 0:EVaultWorldTransitionDirection Enum=Backward
     Option 1:EVaultWorldTransitionDirection Enum=Forward
     Index="integer > integer"(A=@n7, B=@n6)
    ) -> (ReturnValue:EVaultWorldTransitionDirection Enum)) -> (Output_Get:EVaultWorldTransitionDirection Enum=Forward)
   > Set LastItemIndex(LastItemIndex:int=@n7) -> (Output_Get:int=0)
  ]}
 Sequencer Events {nodes [
   BeforeShowArrows() {is=title K2Node_CustomEvent}
   > HandleItemsFadeVisibility() {is=title K2Node_CallFunction}
  ]}
 Set Custom Header Widget {nodes [
   Set Custom Header Widget!() -> (then, NewWidget:User Widget Object Reference) {is=title K2Node_FunctionEntry}
   n1 = GetContent(self=Get CustomContent())
   > n2 = Is Valid(InputObject=@n1)
   RequestLoadBackground(execute=[@n10.else, @n2.Is Not Valid], Direction=Get BackgroundTransitionDirection() -> (BackgroundTransitionDirection=Forward))
   > n3 = SetContent(self=Get CustomContent(), Content=Get NewWidget() -> (NewWidget:User Widget Object Reference))
   n4 = GetContent(self=Get CustomContent())
   n5 = Cast To ItemReceivedCurrencyHeader!(execute=[@n3.then, @n10.then], Object=@n4) -> (then, CastFailed, AsItem Received Currency Header:Item Received Currency Header Object Reference) {is=title K2Node_DynamicCast}
   n6 = GetGiftBox!(self:Gift Box Item VM Object Reference=Get CurrentGiftBoxItem()) -> (ReturnValue:Gift Box VM Object Reference)
   > ActiveFromPresentation!(
    execute
    self:Item Received Currency Header Object Reference=@n5
    CurrencyDataArray:Currency Data With Gift Box Structure=Make Currency Data with Gift Box!(GiftBoxCurrencyData:Array of Currency Data Structures=Get CurrencyData!(self:Gift Box VM Object Reference=@n6) -> (CurrencyData:Array of Currency Data Structures)) -> (CurrencyDataWithGiftBox:Currency Data With Gift Box Structure) {is=title K2Node_MakeStruct}
   ) -> (then)
   n7 = GetExtension(
    self=Cast To UserWidget!(Object=@n4) -> (AsUser Widget:User Widget Object Reference, bSuccess:bool=~) {is=title K2Node_DynamicCast}
    ExtensionType=&MVVMView
   )
   Is Valid(exec=@n5.CastFailed, InputObject=@n7)
   n8 = Get CurrentGiftBoxItem()
   > SetViewModelByClass(self=@n7, NewValue=@n8)
   > SetViewModelByClass(self=@n7, NewValue=GetGiftBox(self=@n8))
   n9 = Get NewWidget()
   n10 = Branch(
    execute=@n2.Is Valid
    Condition="Equal (Class)"(A=Get Class(Object=@n1) -> (ReturnValue:Widget Class Reference), B=Get Class(Object=@n9) -> (ReturnValue:User Widget Class Reference))
   )
  ]}
 SetMobileStyle {nodes [
   SetMobileStyle() {is=title K2Node_FunctionEntry}
   n1 = GetPlayerController!(PlayerIndex:int=0) -> (ReturnValue:Player Controller Object Reference)
   > Set with Broadcast ShouldUseOptimizedTouchLayout!(
    execute
    ShouldUseOptimizedTouchLayout:bool=AND Boolean!(
     A=IsUsingCompactUI()
     B=IsUsingTouch(PlayerController=@n1)
     C:bool=GetConfigDataBoolWithDefault(EntryName=shop_optimized_touch_layout, bDefaultValue=false)
    ) -> (ReturnValue)
   ) -> (then, Output_Get:bool)
   > Branch(Condition=IsUsingCompactUI())
   > Set Visibility(Visibility=Collapsed, self=Get WBP_UIKit_Button_GoToLocker())
  ]}
 ShouldMergeRewardsIntoSingleGiftBox {nodes [
   n1 = ShouldMergeRewardsIntoSingleGiftBox!() -> (then, GiftBoxes:Array of Gift Box VM Object References) {is=title K2Node_FunctionEntry}
   > n2 = For Each Loop(Array=@n1)
   Return Node!(execute=@n2.Completed, IsAllInOneGiftBox:bool=false)
   > Branch(Condition=TextIsEmpty(InText=Get TabText(self=@n2)))
   > Return Node(IsAllInOneGiftBox=true)
  ]}
 StartFade {nodes [
   StartFade!() -> (then, bCloseScreen:bool=~) {is=title K2Node_FunctionEntry}
   > PlayAnimation(InAnimation=Get Anim_Fade())
   > Branch(Condition=Get bCloseScreen() -> (bCloseScreen:bool))
   > BindToAnimationFinished(Animation=Get Anim_Fade(), Delegate=Create Event())
  ]}
 UpdateFrontendCamera {nodes [
   UpdateFrontendCamera() {is=title K2Node_FunctionEntry}
   > Branch(Condition=IsInHestiaBeauty!() -> (ReturnValue:bool))
   > SetFrontEndCamera!(
    execute
    self:Fort Front End Context Object Reference=Get FortFrontEndContext() -> (ReturnValue:Fort Front End Context Object Reference) {is=title K2Node_GetSubsystem}
    NewState:EFrontEndCamera Enum=Vault
    TransitionParams:View Target Transition Params Structure=(BlendTime=0,BlendFunction=VTBlend_Cubic,BlendExp=2,bLockOutgoing=False)
   ) -> (then)
  ]}
 UpdateItemList {nodes [
   n1 = UpdateItemList!() -> (then, GiftBoxVM:Gift Box VM Object Reference) {is=title K2Node_FunctionEntry}
   n2 = Get ListView_Items()
   > Clear Selection(self:List View Object Reference=@n2)
   > ClearListItems(self=@n2)
   > Set List Items!(execute, self:List View Object Reference=Get ListView_Items(), InListItems:Array of Object References=Get Items(self=@n1)) -> (then) {is=title K2Node_CallFunction}
   > n3 = Set LastSelectedIndex(LastSelectedIndex=0) -> (Output_Get=0)
   > SetSelectedIndex(self=Get ListView_Items(), Index=@n3)
   > HandleItemListVisible() {is=title K2Node_CallFunction}
  ]}
 VM_CurrentItemUpdated {nodes [
   VM_CurrentItemUpdated!() -> (then, SelectedItem:Fort Item VM Object Reference) {is=title K2Node_FunctionEntry}
   n1 = Get Slot_ItemCustomBanner() -> (Slot_ItemCustomBanner:Named Slot Object Reference)
   > ClearChildren(self=@n1)
   > SetVisibility(self=@n1, InVisibility=Collapsed)
   n2 = Get SelectedItem() -> (SelectedItem:Fort Item VM Object Reference)
   > Is Valid(InputObject=@n2)
   n3 = TryGetSoftCustomWidgetClass!(
    self:Fort Locker Category Config Object Reference=GetItemCategory(self=@n2)
    ViewRequest:Gameplay Tag Structure=(TagName="UI.ItemPreview.PluginBridge.Mimosa.ShopISB")
   ) -> (WidgetClass:Common User Widget Soft Class Reference, ReturnValue:bool)
   > Branch(Condition=@n3.ReturnValue)
   > n4 = LoadClassAsset_Blocking(AssetClass=@n3)
   > n5 = Cast To CommonUserWidget Class!(execute:Exec, Class:Object Class Reference=@n4) -> (then:Exec, CastFailed:Exec, AsCommon User Widget:Common User Widget Class Reference) {is=title K2Node_ClassDynamicCast}
   > n6 = Create Widget!(execute:Exec, Class:User Widget Class Reference=@n5, OwningPlayer:Player Controller Object Reference=GetOwningPlayer()) -> (then:Exec, ReturnValue:Common User Widget Object Reference)
   > AddChild(self=Get Slot_ItemCustomBanner(), Content=@n6)
   > SetVisibility(self=Get Slot_ItemCustomBanner(), InVisibility=SelfHitTestInvisible)
   > n7 = Cast To WBP_CustomBanner_Base!(execute, Object=@n6) -> (then, CastFailed, AsWBP Custom Banner Base:WBP Custom Banner Base Object Reference) {is=title K2Node_DynamicCast}
   > SetSingleItem!(execute, self:WBP Custom Banner Base Object Reference=@n7, ItemVM:Fort Item VM Object Reference=Get SelectedItem()) -> (then)
   > SetContext!(execute, self:WBP Custom Banner Base Object Reference=@n7, Desired Area:E_MultiproductContext Enum=NewEnumerator2) -> (then)
  ]}
]
variables [
 ActuallySeenRewardsFromGiftBox {type="array<FortItemVM>", default []}
 BackgroundTransitionDirection {type=EVaultWorldTransitionDirection, default=Forward}
 CurrentHeaderWidget {type=UserWidget, default=null}
 Custom Background {type=Object, default=null}
 Input Type {type=ECommonInputType, default=MouseAndKeyboard}
 ItemMinLimit {type=integer, default=1}
 ItemToCustomize {type=FortItemVM, default=null}
 LastItemIndex {type=integer}
 LastSelectedIndex {type=integer}
 Pages Moved {type=integer}
 PendingBackgroundSoftPath {type=SoftObjectPath, default=""}
 PendingTransitionDirection {type=EVaultWorldTransitionDirection, default=Forward}
 PreMessagePushed {type=boolean}
 Primary Content Setup {
  type=PrimaryContentSetup
  default {bEnableNewTextChat=false, bShowBottomBar=true, bShowChatWidget=false, bShowTopBar=false, bTabSupportsInterface=false}
 }
 ShouldGoToLocker {type=boolean, default=true}
 ShouldUseOptimizedTouchLayout {type=boolean}
 WhooshSound {type=SoundBase, default=/Game/Sounds/Fort_UI_Sounds/FrontEnd/RewardsAndGifts/SC_UI_FrontEnd_GiftBox_Screen_Appear_Cue}
]
widgets {
 info {inheritedWidgetCount=0, parentClass=&CommonActivatableWidget, rootWidgetClass=&Overlay, widgetCount=37}
 widgets [
  Overlay_2 {widgetClassPath=&Overlay, widget=Overlay_2}
  SZ_GiftBox {
   widgetClassPath=&SafeZone
   bIsVariable=true
   parent=Overlay_2
   widget=SZ_GiftBox
   properties {padBottom=false, padLeft=false, padRight=false, padTop=false}
   slot {class=OverlaySlot, properties {horizontalAlignment=HAlign_Fill, verticalAlignment=VAlign_Fill}}
  }
  PreMessageSwitcher {
   widgetClassPath=&CommonVisibilitySwitcher
   bIsVariable=true
   parent=SZ_GiftBox
   widget=PreMessageSwitcher
   slot {class=SafeZoneSlot}
  }
  O_GiftBox {
   widgetClassPath=&Overlay
   bIsVariable=true
   parent=PreMessageSwitcher
   widget=O_GiftBox
   slot {class=CommonVisibilitySwitcherSlot}
  }
  FortItemPreviewBackground {
   widgetClassPath=/Script/FortniteUI.FortItemPreviewBackground
   bIsVariable=true
   parent=O_GiftBox
   widget=FortItemPreviewBackground
   properties {previewSceneChangerTag {tagName=UI.ItemPreview.SceneChanger.MPItemShop}}
   slot {class=OverlaySlot, properties {horizontalAlignment=HAlign_Fill, verticalAlignment=VAlign_Fill}}
  }
  PostBufferUpdate_78 {
   widgetClassPath=/Script/UMG.PostBufferUpdate
   bIsVariable=true
   parent=O_GiftBox
   widget=PostBufferUpdate_78
   properties {
    visibility=HitTestInvisible
    updateBufferInfos [
     {bufferToUpdate=ESlatePostRT_0, postParamUpdater=&self:WidgetTree.PostBufferUpdate_78.PostBufferBlurUpdater_0}
     {bufferToUpdate=ESlatePostRT_1, postParamUpdater=null}
    ]
   }
   slot {class=OverlaySlot, properties {horizontalAlignment=HAlign_Fill, verticalAlignment=VAlign_Fill}}
  }
  SizeBox_CameraFrame {
   widgetClassPath=&SizeBox
   parent=O_GiftBox
   widget=SizeBox_CameraFrame
   properties {
    bOverride_HeightOverride=true
    bOverride_WidthOverride=true
    heightOverride=860
    maxDesiredWidth=1200
    minDesiredWidth=1024
    widthOverride=760
   }
   slot {class=OverlaySlot, properties {horizontalAlignment=HAlign_Center, verticalAlignment=VAlign_Center, padding=<ltrb 0 60 0 110>}}
  }
  FortItemPreviewWidget {
   widgetClassPath=/FortniteCoreUI/Frontend/ItemPreview/WBP_ItemPreview
   bIsVariable=true
   parent=SizeBox_CameraFrame
   widget=FortItemPreviewWidget
   properties {
    bAllowPawnOverrideOnExclusiveItems=true
    bCanDisplayUserAccessories=false
    bShowFloor=false
    bSupportsActivationFocus=false
    itemViewObject=&self:WidgetTree.FortItemPreviewWidget.ItemViewObject
    previewActionButtons [null]
    previewSceneChangerTag {tagName=UI.ItemPreview.SceneChanger.MPItemShop}
   }
   slot {class=SizeBoxSlot}
  }
  Txt_Equipped {
   widgetClassPath=&CommonTextBlock
   parent=O_GiftBox
   widget=Txt_Equipped
   properties {
    justification=Center
    renderOpacity=0
    text=EQUIPPED
    visibility=Collapsed
    colorAndOpacity {colorUseRule=UseColor_Specified, specifiedColor=<rgba 0 0.6654 1 1>}
    font {
     bForceMonospaced=false
     bMaterialIsStencil=false
     fontMaterial=null
     fontObject=&HeadingNow
     letterSpacing=0
     monospacedWidth=1
     size=36
     skewAmount=0
     typefaceFontName=86 Bold
     outlineSettings {
      bApplyOutlineToDropShadows=false
      bMiteredCorners=false
      bSeparateFillAlpha=false
      outlineMaterial=null
      outlineSize=0
      outlineColor=<rgba 0 0 0 1>
     }
    }
    shadowOffset=<0 0>
   }
   slot {class=OverlaySlot, properties {horizontalAlignment=HAlign_Center, verticalAlignment=VAlign_Center}}
  }
  VB_Content {
   widgetClassPath=&VerticalBox
   parent=O_GiftBox
   widget=VB_Content
   slot {class=OverlaySlot, properties {horizontalAlignment=HAlign_Fill, verticalAlignment=VAlign_Fill, padding=<ltrb 0 32 0 64>}}
  }
  WBP_UIKit_TabGroup {
   widgetClassPath=/FortUILibrary/UIKit/Tabs/WBP_UIKit_TabGroup
   bIsVariable=true
   parent=VB_Content
   widget=WBP_UIKit_TabGroup
   properties {
    backplateBrightness=Dark
    buttonClass=&/GiftBox_Button_TopNav.GiftBox_Button_TopNav_C
    tabSpacing=8
    usePromptBackplate=false
    initialTabsData [
     {bEnableTabId=false, tabId=None, tabViewModel=&self:WidgetTree.WBP_UIKit_TabGroup.UIKitSelectableButtonViewModel_0}
     {bEnableTabId=false, tabId=None, tabViewModel=&self:WidgetTree.WBP_UIKit_TabGroup.UIKitSelectableButtonViewModel_2}
     {bEnableTabId=false, tabId=None, tabViewModel=&self:WidgetTree.WBP_UIKit_TabGroup.UIKitSelectableButtonViewModel_1}
    ]
   }
   slot {class=VerticalBoxSlot, properties {horizontalAlignment=HAlign_Center, verticalAlignment=VAlign_Top}}
  }
  HB_Panels {
   widgetClassPath=&HorizontalBox
   parent=VB_Content
   widget=HB_Panels
   slot {class=VerticalBoxSlot, properties {padding=<ltrb 80 0 80 0>, size {sizeRule=Fill, value=1}}}
  }
  LeftsideContentSwitcher {
   widgetClassPath=&WidgetSwitcher
   bIsVariable=true
   parent=HB_Panels
   widget=LeftsideContentSwitcher
   slot {class=HorizontalBoxSlot, properties {size {sizeRule=Fill, value=0.85}}}
  }
  DefaultContent {
   widgetClassPath=&Overlay
   bIsVariable=true
   parent=LeftsideContentSwitcher
   widget=DefaultContent
   slot {class=WidgetSwitcherSlot, properties {verticalAlignment=VAlign_Top, padding=<ltrb 0 260 0 0>}}
  }
  ItemHeaderWidget {
   widgetClassPath=&/GiftBoxScreenHeader
   bIsVariable=true
   parent=DefaultContent
   widget=ItemHeaderWidget
   slot {class=OverlaySlot, properties {horizontalAlignment=HAlign_Fill, verticalAlignment=VAlign_Center}}
  }
  Overlay_CustomContent {
   widgetClassPath=&Overlay
   bIsVariable=true
   parent=LeftsideContentSwitcher
   widget=Overlay_CustomContent
   slot {class=WidgetSwitcherSlot, properties {verticalAlignment=VAlign_Top, padding=<ltrb 0 260 0 0>}}
  }
  CustomContent {
   widgetClassPath=&NamedSlot
   bIsVariable=true
   parent=Overlay_CustomContent
   widget=CustomContent
   properties {bExposeOnInstanceOnly=true}
   slot {class=OverlaySlot, properties {horizontalAlignment=HAlign_Fill, verticalAlignment=VAlign_Center}}
  }
  Spacer_ItemPreview {
   widgetClassPath=&Spacer
   parent=HB_Panels
   widget=Spacer_ItemPreview
   slot {class=HorizontalBoxSlot, properties {size {sizeRule=Fill, value=1}}}
  }
  VB_RightPanel {
   widgetClassPath=&VerticalBox
   parent=HB_Panels
   widget=VB_RightPanel
   slot {class=HorizontalBoxSlot, properties {horizontalAlignment=HAlign_Right, size {sizeRule=Fill, value=0.85}}}
  }
  ItemDescription {
   widgetClassPath=/FortUILibrary/UIKit/Items/WBP_UIKit_DescriptionWithSubtitle
   bIsVariable=true
   parent=VB_RightPanel
   widget=ItemDescription
   properties {
    bodyTextCustomTextStyle=/Game/Athena/Items/Cosmetics/Metadata/ItemDescriptionTextStyle
    bodyTextRow=3
    bottomSlotRow=5
    compatibilityRow=4
    hideBodyText=false
    hideCompatibility=true
    hideGameplayTagsText=false
    hideSetText=false
    hideSubtitle=true
    hideVarianceText=false
    tagsOrientation=Orient_Vertical
    use Rich Description=false
    compatibilityPadding=<ltrb 0 0 0 0>
   }
   slot {class=VerticalBoxSlot, properties {padding=<ltrb 0 260 0 0>}}
  }
  Slot_GiftboxCustomBanner {
   widgetClassPath=&NamedSlot
   bIsVariable=true
   parent=VB_RightPanel
   widget=Slot_GiftboxCustomBanner
   properties {bExposeOnInstanceOnly=true}
   slot {class=VerticalBoxSlot, properties {verticalAlignment=VAlign_Bottom, padding=<ltrb 0 16 0 0>}}
  }
  Slot_ItemCustomBanner {
   widgetClassPath=&NamedSlot
   bIsVariable=true
   parent=VB_RightPanel
   widget=Slot_ItemCustomBanner
   properties {bExposeOnInstanceOnly=true}
   slot {class=VerticalBoxSlot, properties {verticalAlignment=VAlign_Bottom, padding=<ltrb 0 16 0 0>}}
  }
  GP_ItemList {
   widgetClassPath=&GridPanel
   bIsVariable=true
   parent=VB_Content
   widget=GP_ItemList
   properties {columnFill [0,1,0], rowFill [1]}
   slot {class=VerticalBoxSlot, properties {horizontalAlignment=HAlign_Center, verticalAlignment=VAlign_Top, padding=<ltrb 80 0 80 0>}}
  }
  O_ScrollingListContainer {
   widgetClassPath=&Overlay
   parent=GP_ItemList
   widget=O_ScrollingListContainer
   properties {clipping=ClipToBounds}
   slot {class=GridSlot, properties {column=1, padding=<ltrb -80 0 -80 0>}}
  }
  ListView_Items {
   widgetClassPath=&CommonListView
   bIsVariable=true
   parent=O_ScrollingListContainer
   widget=ListView_Items
   properties {
    bInEnableTouchAnimatedScrolling=true
    bIsFocusable=false
    clipping=Inherit
    entryWidgetClass=&/WBP_UIKit_ItemCard_GiftBoxItem.WBP_UIKit_ItemCard_GiftBoxItem_C
    horizontalEntrySpacing=16
    numDesignerPreviewEntries=20
    orientation=Orient_Horizontal
    scrollBarStyle {
     thickness=0
     draggedThumbImage {
      bIsDynamicallyLoaded=false
      drawAs=Box
      imageType=FullColor
      mirroring=NoMirror
      resourceName=../../../Engine/Content/Slate/Common/Scrollbar_Thumb.png
      resourceObject=null
      tiling=NoTile
      imageSize=<0 0>
      margin=<ltrb 0.25 0.25 0.25 0.25>
      outlineSettings {
       bUseBrushTransparency=false
       roundingType=HalfHeightRadius
       width=0
       color {colorUseRule=UseColor_Specified, specifiedColor=<rgba 0 0 0 0>}
       cornerRadii=<0 0 0 0>
      }
      tintColor {colorUseRule=UseColor_Specified, specifiedColor=<rgba 1 1 1 1>}
      uVRegion {bIsValid=false, max=<0 0>, min=<0 0>}
     }
     horizontalBackgroundImage {
      bIsDynamicallyLoaded=false
      drawAs=NoDrawType
      imageType=NoImage
      mirroring=NoMirror
      resourceName=None
      resourceObject=null
      tiling=NoTile
      imageSize=<0 0>
      margin=<ltrb 0 0 0 0>
      outlineSettings {
       bUseBrushTransparency=false
       roundingType=HalfHeightRadius
       width=0
       color {colorUseRule=UseColor_Specified, specifiedColor=<rgba 0 0 0 0>}
       cornerRadii=<0 0 0 0>
      }
      tintColor {colorUseRule=UseColor_Specified, specifiedColor=<rgba 1 1 1 1>}
      uVRegion {bIsValid=false, max=<0 0>, min=<0 0>}
     }
     horizontalBottomSlotImage {
      bIsDynamicallyLoaded=false
      drawAs=Image
      imageType=FullColor
      mirroring=NoMirror
      resourceName=../../../Engine/Content/Slate/Common/Scrollbar_Background_Horizontal.png
      resourceObject=null
      tiling=NoTile
      imageSize=<8 8>
      margin=<ltrb 0 0 0 0>
      outlineSettings {
       bUseBrushTransparency=false
       roundingType=HalfHeightRadius
       width=0
       color {colorUseRule=UseColor_Specified, specifiedColor=<rgba 0 0 0 0>}
       cornerRadii=<0 0 0 0>
      }
      tintColor {colorUseRule=UseColor_Specified, specifiedColor=<rgba 1 1 1 1>}
      uVRegion {bIsValid=false, max=<0 0>, min=<0 0>}
     }
     horizontalTopSlotImage {
      bIsDynamicallyLoaded=false
      drawAs=Image
      imageType=FullColor
      mirroring=NoMirror
      resourceName=../../../Engine/Content/Slate/Common/Scrollbar_Background_Horizontal.png
      resourceObject=null
      tiling=NoTile
      imageSize=<8 8>
      margin=<ltrb 0 0 0 0>
      outlineSettings {
       bUseBrushTransparency=false
       roundingType=HalfHeightRadius
       width=0
       color {colorUseRule=UseColor_Specified, specifiedColor=<rgba 0 0 0 0>}
       cornerRadii=<0 0 0 0>
      }
      tintColor {colorUseRule=UseColor_Specified, specifiedColor=<rgba 1 1 1 1>}
      uVRegion {bIsValid=false, max=<0 0>, min=<0 0>}
     }
     hoveredThumbImage {
      bIsDynamicallyLoaded=false
      drawAs=Box
      imageType=FullColor
      mirroring=NoMirror
      resourceName=../../../Engine/Content/Slate/Common/Scrollbar_Thumb.png
      resourceObject=null
      tiling=NoTile
      imageSize=<0 0>
      margin=<ltrb 0.25 0.25 0.25 0.25>
      outlineSettings {
       bUseBrushTransparency=false
       roundingType=HalfHeightRadius
       width=0
       color {colorUseRule=UseColor_Specified, specifiedColor=<rgba 0 0 0 0>}
       cornerRadii=<0 0 0 0>
      }
      tintColor {colorUseRule=UseColor_Specified, specifiedColor=<rgba 1 1 1 1>}
      uVRegion {bIsValid=false, max=<0 0>, min=<0 0>}
     }
     normalThumbImage {
      bIsDynamicallyLoaded=false
      drawAs=Box
      imageType=FullColor
      mirroring=NoMirror
      resourceName=../../../Engine/Content/Slate/Common/Scrollbar_Thumb.png
      resourceObject=null
      tiling=NoTile
      imageSize=<0 0>
      margin=<ltrb 0.25 0.25 0.25 0.25>
      outlineSettings {
       bUseBrushTransparency=false
       roundingType=HalfHeightRadius
       width=0
       color {colorUseRule=UseColor_Specified, specifiedColor=<rgba 0 0 0 0>}
       cornerRadii=<0 0 0 0>
      }
      tintColor {colorUseRule=UseColor_Specified, specifiedColor=<rgba 1 1 1 1>}
      uVRegion {bIsValid=false, max=<0 0>, min=<0 0>}
     }
     verticalBackgroundImage {
      bIsDynamicallyLoaded=false
      drawAs=NoDrawType
      imageType=NoImage
      mirroring=NoMirror
      resourceName=None
      resourceObject=null
      tiling=NoTile
      imageSize=<0 0>
      margin=<ltrb 0 0 0 0>
      outlineSettings {
       bUseBrushTransparency=false
       roundingType=HalfHeightRadius
       width=0
       color {colorUseRule=UseColor_Specified, specifiedColor=<rgba 0 0 0 0>}
       cornerRadii=<0 0 0 0>
      }
      tintColor {colorUseRule=UseColor_Specified, specifiedColor=<rgba 1 1 1 1>}
      uVRegion {bIsValid=false, max=<0 0>, min=<0 0>}
     }
     verticalBottomSlotImage {
      bIsDynamicallyLoaded=false
      drawAs=Image
      imageType=FullColor
      mirroring=NoMirror
      resourceName=../../../Engine/Content/Slate/Common/Scrollbar_Background_Vertical.png
      resourceObject=null
      tiling=NoTile
      imageSize=<8 8>
      margin=<ltrb 0 0 0 0>
      outlineSettings {
       bUseBrushTransparency=false
       roundingType=HalfHeightRadius
       width=0
       color {colorUseRule=UseColor_Specified, specifiedColor=<rgba 0 0 0 0>}
       cornerRadii=<0 0 0 0>
      }
      tintColor {colorUseRule=UseColor_Specified, specifiedColor=<rgba 1 1 1 1>}
      uVRegion {bIsValid=false, max=<0 0>, min=<0 0>}
     }
     verticalTopSlotImage {
      bIsDynamicallyLoaded=false
      drawAs=Image
      imageType=FullColor
      mirroring=NoMirror
      resourceName=../../../Engine/Content/Slate/Common/Scrollbar_Background_Vertical.png
      resourceObject=null
      tiling=NoTile
      imageSize=<8 8>
      margin=<ltrb 0 0 0 0>
      outlineSettings {
       bUseBrushTransparency=false
       roundingType=HalfHeightRadius
       width=0
       color {colorUseRule=UseColor_Specified, specifiedColor=<rgba 0 0 0 0>}
       cornerRadii=<0 0 0 0>
      }
      tintColor {colorUseRule=UseColor_Specified, specifiedColor=<rgba 1 1 1 1>}
      uVRegion {bIsValid=false, max=<0 0>, min=<0 0>}
     }
    }
   }
   slot {class=OverlaySlot, properties {horizontalAlignment=HAlign_Fill, verticalAlignment=VAlign_Fill, padding=<ltrb 80 32 80 24>}}
  }
  LeftFade {
   widgetClassPath=&Image
   bIsVariable=true
   parent=O_ScrollingListContainer
   widget=LeftFade
   properties {
    visibility=HitTestInvisible
    brush {
     bIsDynamicallyLoaded=false
     drawAs=Image
     imageType=FullColor
     mirroring=NoMirror
     resourceName=None
     resourceObject=/GiftBox/UI/Art/MaterialInstances/MI_UI_GiftBox_ListFade_Left
     tiling=NoTile
     imageSize=<80 32>
     margin=<ltrb 0 0 0 0>
     outlineSettings {
      bUseBrushTransparency=false
      roundingType=HalfHeightRadius
      width=0
      color {colorUseRule=UseColor_Specified, specifiedColor=<rgba 0 0 0 0>}
      cornerRadii=<0 0 0 0>
     }
     tintColor {colorUseRule=UseColor_Specified, specifiedColor=<rgba 1 1 1 1>}
     uVRegion {bIsValid=false, max=<0 0>, min=<0 0>}
    }
   }
   slot {class=OverlaySlot, properties {verticalAlignment=VAlign_Fill}}
  }
  RightFade {
   widgetClassPath=&Image
   bIsVariable=true
   parent=O_ScrollingListContainer
   widget=RightFade
   properties {
    visibility=HitTestInvisible
    brush {
     bIsDynamicallyLoaded=false
     drawAs=Image
     imageType=FullColor
     mirroring=NoMirror
     resourceName=None
     resourceObject=/GiftBox/UI/Art/MaterialInstances/MI_UI_GiftBox_ListFade_Right
     tiling=NoTile
     imageSize=<80 32>
     margin=<ltrb 0 0 0 0>
     outlineSettings {
      bUseBrushTransparency=false
      roundingType=HalfHeightRadius
      width=0
      color {colorUseRule=UseColor_Specified, specifiedColor=<rgba 0 0 0 0>}
      cornerRadii=<0 0 0 0>
     }
     tintColor {colorUseRule=UseColor_Specified, specifiedColor=<rgba 1 1 1 1>}
     uVRegion {bIsValid=false, max=<0 0>, min=<0 0>}
    }
   }
   slot {class=OverlaySlot, properties {horizontalAlignment=HAlign_Right, verticalAlignment=VAlign_Fill}}
  }
  Button_LeftScrollArrow {
   widgetClassPath=&WBP_UIKit_Button_Regular
   bIsVariable=true
   parent=GP_ItemList
   widget=Button_LeftScrollArrow
   properties {
    bHideInputAction=true
    bIsFocusable=false
    clipping=OnDemand
    iconMaterial=/GiftBox/UI/Art/MaterialInstances/MI_Giftbox_ArrowLeft
    text=""
    visibility=Visible
   }
   slot {class=GridSlot, properties {horizontalAlignment=HAlign_Left, layer=10, verticalAlignment=VAlign_Center, padding=<ltrb 0 0 24 0>}}
  }
  Button_RightScrollArrow {
   widgetClassPath=&WBP_UIKit_Button_Regular
   bIsVariable=true
   parent=GP_ItemList
   widget=Button_RightScrollArrow
   properties {bHideInputAction=true, bIsFocusable=false, iconMaterial=/GiftBox/UI/Art/MaterialInstances/MI_Giftbox_ArrowRight, text="", visibility=Visible}
   slot {class=GridSlot, properties {column=2, horizontalAlignment=HAlign_Right, layer=10, verticalAlignment=VAlign_Center, padding=<ltrb 24 0 0 0>}}
  }
  O_Buttons {
   widgetClassPath=&Overlay
   bIsVariable=true
   parent=VB_Content
   widget=O_Buttons
   slot {class=VerticalBoxSlot, properties {verticalAlignment=VAlign_Top, padding=<ltrb 80 0 80 0>}}
  }
  HB_BtnActions {
   widgetClassPath=&HorizontalBox
   bIsVariable=true
   parent=O_Buttons
   widget=HB_BtnActions
   slot {class=OverlaySlot, properties {horizontalAlignment=HAlign_Center}}
  }
  Btn_LeftAction {
   widgetClassPath=&WBP_UIKit_Button_Regular
   bIsVariable=true
   parent=HB_BtnActions
   widget=Btn_LeftAction
   properties {
    bIsFocusable=false
    minWidth=304
    mobileMinWidth=192
    text=Equip
    triggeringInputAction {dataTable=&InputActionDataTable, rowName=GenericFaceButtonLeft}
   }
   slot {class=HorizontalBoxSlot, properties {verticalAlignment=VAlign_Top, padding=<ltrb 0 0 32 0>}}
  }
  Btn_RightAction {
   widgetClassPath=&WBP_UIKit_Button_Regular
   bIsVariable=true
   parent=HB_BtnActions
   widget=Btn_RightAction
   properties {
    bIsFocusable=false
    minWidth=304
    mobileMinWidth=192
    text=Claim All
    triggeringInputAction {dataTable=&InputActionDataTable, rowName=Confirm}
   }
   slot {class=HorizontalBoxSlot, properties {verticalAlignment=VAlign_Top}}
  }
  Text_ItemsCollected {
   widgetClassPath=&CommonTextBlock
   parent=O_Buttons
   widget=Text_ItemsCollected
   properties {
    justification=Center
    text=X Items Collected
    textTransformPolicy=ToUpper
    visibility=Hidden
    colorAndOpacity {colorUseRule=UseColor_Specified, specifiedColor=<rgba 0.9911 0.9911 0.9911 1>}
    font {
     bForceMonospaced=false
     bMaterialIsStencil=false
     fontMaterial=null
     fontObject=&HeadingNow
     letterSpacing=0
     monospacedWidth=1
     size=32
     skewAmount=0
     typefaceFontName=86 Bold
     outlineSettings {
      bApplyOutlineToDropShadows=false
      bMiteredCorners=false
      bSeparateFillAlpha=false
      outlineMaterial=null
      outlineSize=0
      outlineColor=<rgba 0 0 0 1>
     }
    }
    shadowOffset=<0 0>
   }
   slot {class=OverlaySlot, properties {horizontalAlignment=HAlign_Center}}
  }
  WBP_UIKit_Button_GoToLocker {
   widgetClassPath=/FortUILibrary/UIKit/Buttons/WBP_UIKit_Button_Quiet
   bIsVariable=true
   parent=O_Buttons
   widget=WBP_UIKit_Button_GoToLocker
   properties {
    text=Go to Locker
    visibility=Collapsed
    triggeringInputAction {dataTable=&InputActionDataTable, rowName=GenericFaceButtonTop}
   }
   slot {class=OverlaySlot, properties {verticalAlignment=VAlign_Center}}
  }
  Overlay_0 {
   widgetClassPath=&Overlay
   bIsVariable=true
   parent=PreMessageSwitcher
   widget=Overlay_0
   slot {class=CommonVisibilitySwitcherSlot}
  }
  PreMessageSlot {
   widgetClassPath=&NamedSlot
   bIsVariable=true
   parent=Overlay_0
   widget=PreMessageSlot
   slot {class=OverlaySlot, properties {horizontalAlignment=HAlign_Fill, verticalAlignment=VAlign_Fill}}
  }
 ]
}
