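# WBP_GiftBox_BonusInfo is a UserWidget for the GiftBox UI that shows a bonus info panel with a status badge.
# A GridPanel layers auto-wrapping, muted grey text over a dark UIKit backplate.
# A FortVisualAttachment with a centered content anchor pins a compact UIKit status indicator to the top-left corner.
# The text and indicator are exposed as variables, while the event graph holds only empty default events.
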WidgetBlueprint /GiftBox/UI/WBP_GiftBox_BonusInfo
graphs [EventGraph {nodes [
   Event PreConstruct()
   Event Construct()
   Event Tick()
  ]}]
variables []
widgets {
 info {inheritedWidgetCount=0, parentClass=&UserWidget, rootWidgetClass=&GridPanel, widgetCount=5}
 widgets [
  GP_Content {widgetClassPath=&GridPanel, widget=GP_Content, properties {columnFill [1], rowFill [1]}}
  WBP_UIKit_Backplate {
   widgetClassPath=&WBP_UIKit_Backplate
   parent=GP_Content
   widget=WBP_UIKit_Backplate
   properties {brightness=Dark}
   slot {class=GridSlot}
  }
  Txt_Info {
   widgetClassPath=&CommonTextBlock
   bIsVariable=true
   parent=GP_Content
   widget=Txt_Info
   properties {
    autoWrapText=true
    text=$Info Text
    colorAndOpacity {colorUseRule=UseColor_Specified, specifiedColor=<rgba 0.5776 0.5776 0.5776 1>}
    font {
     bForceMonospaced=false
     bMaterialIsStencil=false
     fontMaterial=null
     fontObject=&NotoSans
     letterSpacing=0
     monospacedWidth=1
     size=24
     skewAmount=0
     typefaceFontName=NotoSans-Regular
     outlineSettings {
      bApplyOutlineToDropShadows=false
      bMiteredCorners=false
      bSeparateFillAlpha=false
      outlineMaterial=null
      outlineSize=0
      outlineColor=<rgba 0 0 0 1>
     }
    }
    shadowOffset=<0 0>
   }
   slot {class=GridSlot, properties {layer=10, verticalAlignment=VAlign_Top, padding=<ltrb 12 8 12 8>}}
  }
  VA_StatusIndicator {
   widgetClassPath=&FortVisualAttachment
   parent=GP_Content
   widget=VA_StatusIndicator
   properties {contentAnchor=<0.5 0.5>}
   slot {class=GridSlot, properties {horizontalAlignment=HAlign_Left, layer=20, verticalAlignment=VAlign_Top}}
  }
  WBP_UIKit_StatusIndicator {
   widgetClassPath=/FortUILibrary/UIKit/StatusIndicator/WBP_UIKit_StatusIndicator
   bIsVariable=true
   parent=VA_StatusIndicator
   widget=WBP_UIKit_StatusIndicator
   properties {iconSize=Compact}
   slot {class=SizeBoxSlot, properties {horizontalAlignment=HAlign_Center, verticalAlignment=VAlign_Center}}
  }
 ]
}
